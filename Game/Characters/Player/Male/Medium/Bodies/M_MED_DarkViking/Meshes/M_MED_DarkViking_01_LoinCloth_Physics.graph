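# Physics asset for the Dark Viking medium male loincloth mesh, likely serving as collision volumes for the cloth.
# It covers only the lower body with capsule shapes on the pelvis and both thighs, using two capsules each for pelvis and left thigh and one for the right.
# Each thigh is constrained to the pelvis with swing and twist all free, so the joints impose no angular limits.
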PhysicsAsset /Game/Characters/Player/Male/Medium/Bodies/M_MED_DarkViking/Meshes/M_MED_DarkViking_01_LoinCloth_Physics
bodies {
 pelvis [
  {shapeType=Capsule, length=6.5, radius=8.6099, center=<-9.5081 2.4996 -0.0001>}
  {shapeType=Capsule, length=31, radius=9.8599, center=<12.2742 -3.8387 -0.0001>, rotation=<rot 87.9444 87.9433 89.9962>}
 ]
 thigh_l [
  {shapeType=Capsule, length=21.2445, radius=10.1311, center=<-15.4159 0.1006 -1.126>, rotation=<rot -89.6534 90.0564 -95.3513>}
  {shapeType=Capsule, length=19.3756, radius=4.5, center=<-12.5392 6.4205 2.5335>, rotation=<rot -89.6534 90.0564 -95.3513>}
 ]
 thigh_r [{shapeType=Capsule, length=26.0769, radius=9.4949, center=<17.8999 -1.0541 0.3038>, rotation=<rot 89.6631 89.729 90.0245>}]
}
constraints [
 {
  bone1Name=thigh_l
  bone2Name=pelvis
  swing1LimitDegrees=45
  swing1Motion=Free
  swing2LimitDegrees=45
  swing2Motion=Free
  twistLimitDegrees=45
  twistMotion=Free
 }
 {
  bone1Name=thigh_r
  bone2Name=pelvis
  swing1LimitDegrees=45
  swing1Motion=Free
  swing2LimitDegrees=45
  swing2Motion=Free
  twistLimitDegrees=45
  twistMotion=Free
 }
]
