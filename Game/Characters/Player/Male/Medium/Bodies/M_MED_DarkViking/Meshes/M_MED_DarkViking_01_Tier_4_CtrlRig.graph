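# Control Rig for the Tier 4 medium male Dark Viking player body that applies a pose-driven correction to the bandolier bone.
# A spherical pose reader measures the left upper arm's X axis against an offset active region with falloff.
# Its output weights a Modify Transforms unit that shifts, rotates and Z-squashes dyn_bandolier_1, a child of the left clavicle.
# The hierarchy carries the full body, face and dyn_ bones plus curves, but the graph modifies only that bone.
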ControlRigBlueprint /Game/Characters/Player/Male/Medium/Bodies/M_MED_DarkViking/Meshes/M_MED_DarkViking_01_Tier_4_CtrlRig
graphs [
 RigVMFunctionLibrary {links [], nodes []}
 RigVMModel {nodes [
   RigUnit_BeginExecution = &RigUnit_BeginExecution()
   n1 = &RigUnit_SphericalPoseReader(
    ActiveRegionScaleFactors=(PositiveWidth=1.5,NegativeWidth=1.5,PositiveHeight=1.5,NegativeHeight=1.5)
    ActiveRegionSize=0.1
    Debug=(bDrawDebug=false,bDrawLocalAxes=false,DebugScale=25,DebugSegments=20,DebugThickness=0.25)
    DriverAxis=(X=1,Y=0,Z=0)
    DriverItem=(Type=Bone,Name="upperarm_l")
    FalloffRegionScaleFactors=(PositiveWidth=1,NegativeWidth=1,PositiveHeight=1,NegativeHeight=1)
    FalloffSize=0.35
    FlipWidthScaling=false
    OptionalParentItem=(Type=Bone)
    RotationOffset=(X=30,Y=-30,Z=0)
    ExecutePin=@RigUnit_BeginExecution.ExecutePin
   )
   > &RigUnit_ModifyTransforms(
    ItemToModify=((Item=(Type=Bone,Name="dyn_bandolier_1"),Transform=(Rotation=(X=-0.0872,Y=0,Z=0,W=-0.9962),Translation=(X=-0.5,Y=0,Z=-3),Scale3D=(X=1,Y=1,Z=0.7))))
    WeightMinimum=0
    Weight=@n1.OutputParam
   )
  ]}
]
hierarchy [
 root {loc=<0 0 0>, rot=<0 0 0 1>}
 attach {parent=root, loc=<0 0 0>, rot=<0 0 0 1>}
 pelvis {parent=root, loc=<0 -0.7071 93.1191>, rot=<0.0127 0.707 -0.0127 -0.707>}
 dyn_pouch_r {parent=pelvis, loc=<0 0 0>, rot=<0.7071 -0.0013 -0.7071 -0.0013>}
 spine_01 {parent=pelvis, loc=<6.7494 -0.6388 0>, rot=<0 0 0.034 -0.9994>}
 spine_02 {parent=spine_01, loc=<7.2518 -0.035 0>, rot=<0 0 0.0238 -0.9997>}
 spine_03 {parent=spine_02, loc=<6.0104 0 0>, rot=<0 0 0.0149 -0.9999>}
 spine_04 {parent=spine_03, loc=<8.1024 0 0>, rot=<0 0 -0.0132 -0.9999>}
 spine_05 {parent=spine_04, loc=<10.0146 0 0>, rot=<0 0 -0.0015 -1>}
 clavicle_l {parent=spine_05, loc=<4.3575 -0.1466 -2.3223>, rot=<0.0465 -0.7289 -0.009 -0.6829>}
 upperarm_l {parent=clavicle_l, loc=<15.5111 0 0>, rot=<0.0305 -0.536 -0.0057 -0.8437>}
 lowerarm_l {parent=upperarm_l, loc=<26.6514 0 0>, rot=<0.0391 -0.0099 -0.2453 -0.9686>}
 hand_l {parent=lowerarm_l, loc=<22.3491 0 0>, rot=<-0.7081 -0.0507 0.0975 -0.6975>}
 index_metacarpal_l {parent=hand_l, loc=<1.6535 0.0183 -2.96>, rot=<0.0158 -0.0124 -0.0173 -0.9996>}
 index_01_l {parent=index_metacarpal_l, loc=<8.3008 0.0357 -0.0524>, rot=<-0.0044 0.0485 0.0924 -0.9945>}
 index_02_l {parent=index_01_l, loc=<3.3587 -0.3004 0.0452>, rot=<-0.002 0.0027 0.3225 -0.9466>}
 index_03_l {parent=index_02_l, loc=<2.5578 -0.123 -0.0252>, rot=<-0.002 0.0067 0.2558 -0.9667>}
 middle_metacarpal_l {parent=hand_l, loc=<2.1856 -0.1725 -0.5984>, rot=<0.006 0.0168 -0.0329 -0.9993>}
 middle_01_l {parent=middle_metacarpal_l, loc=<7.3275 -0.2447 0.0129>, rot=<-0.0208 0.0042 0.0858 -0.9961>}
 middle_02_l {parent=middle_01_l, loc=<4.0123 0.1401 0.1175>, rot=<-0.0218 0.0258 0.3512 -0.9357>}
 middle_03_l {parent=middle_02_l, loc=<2.6077 0.0534 0.1373>, rot=<0.0082 0.0188 0.1708 -0.9851>}
 pinky_metacarpal_l {parent=hand_l, loc=<1.7616 -0.9753 3.3814>, rot=<-0.0397 0.0814 -0.069 -0.9935>}
 pinky_01_l {parent=pinky_metacarpal_l, loc=<7.1709 -0.4831 -0.1094>, rot=<-0.0654 -0.0035 0.2242 -0.9723>}
 pinky_02_l {parent=pinky_01_l, loc=<2.8274 0.0893 -0.1043>, rot=<0.0037 0.0011 0.2854 -0.9584>}
 pinky_03_l {parent=pinky_02_l, loc=<2.2046 0.0307 -0.0078>, rot=<-0.0002 0.0098 0.1445 -0.9895>}
 ring_metacarpal_l {parent=hand_l, loc=<2.2074 -0.3625 1.3638>, rot=<-0.0284 0.0722 -0.0458 -0.9959>}
 ring_01_l {parent=ring_metacarpal_l, loc=<7.0295 -0.6197 -0.1446>, rot=<-0.0489 -0.0252 0.1892 -0.9804>}
 ring_02_l {parent=ring_01_l, loc=<3.6174 0.4647 -0.1027>, rot=<0 0 0.2806 -0.9598>}
 ring_03_l {parent=ring_02_l, loc=<2.4152 0.114 0.0199>, rot=<0 0 0.1189 -0.9929>}
 thumb_01_l {parent=hand_l, loc=<1.3118 -0.3191 -2.459>, rot=<0.6546 -0.4042 0.0664 -0.6354>}
 thumb_02_l {parent=thumb_01_l, loc=<5.9336 -0.0448 0.0401>, rot=<-0.0037 0.0213 0.298 -0.9543>}
 thumb_03_l {parent=thumb_02_l, loc=<2.2223 0.3682 -0.0907>, rot=<0.0118 -0.013 0.0066 -0.9998>}
 weapon_l {parent=hand_l, loc=<-0.1974 -3.4111 0.3811>, rot=<0 0 0 1>}
 lowerarm_twist_01_l {parent=lowerarm_l, loc=<17.8549 0 0>, rot=<0 0 0 1>}
 lowerarm_twist_02_l {parent=lowerarm_l, loc=<9.6199 0 0>, rot=<0 0 0 1>}
 upperarm_twist_01_l {parent=upperarm_l, loc=<0.0041 0 0>, rot=<0 0 0 1>}
 upperarm_twist_02_l {parent=upperarm_l, loc=<13.7941 0 0>, rot=<0 0 0 1>}
 dyn_shoulder_armor_LOCALSPACE_l {parent=upperarm_l, loc=<0 0 0>, rot=<0 0 0 1>}
 dyn_shoulder_armor_driver_l {parent=dyn_shoulder_armor_LOCALSPACE_l, loc=<-7.3563 1.0385 -4.9827>, rot=<0.025 0.5646 0.0141 -0.8249>}
 dyn_shoulder_armor_driver_B_l {parent=dyn_shoulder_armor_driver_l, loc=<16.5235 0 -14.5467>, rot=<0 0 0 1>}
 dyn_shoulder_armor_l {parent=dyn_shoulder_armor_driver_B_l, loc=<-16.5235 0 14.5467>, rot=<0 0 0 1>}
 dyn_bandolier_1 {parent=clavicle_l, loc=<8.6433 4.0925 5.2014>, rot=<0.1561 0.4666 -0.3186 -0.8102>}
 clavicle_r {parent=spine_05, loc=<4.3575 -0.1466 2.3223>, rot=<0.7289 0.0465 -0.6829 0.009>}
 upperarm_r {parent=clavicle_r, loc=<-15.5111 0 0>, rot=<0.0305 -0.536 -0.0057 -0.8437>}
 lowerarm_r {parent=upperarm_r, loc=<-26.6515 0 0>, rot=<0.0391 -0.0099 -0.2453 -0.9686>}
 hand_r {parent=lowerarm_r, loc=<-22.3492 0 0>, rot=<-0.7081 -0.0507 0.0975 -0.6975>}
 index_metacarpal_r {parent=hand_r, loc=<-1.6534 -0.0181 2.9601>, rot=<0.0158 -0.0124 -0.0173 -0.9996>}
 index_01_r {parent=index_metacarpal_r, loc=<-8.3008 -0.0357 0.0524>, rot=<-0.0044 0.0485 0.0924 -0.9945>}
 index_02_r {parent=index_01_r, loc=<-3.3589 0.3003 -0.0453>, rot=<-0.002 0.0027 0.3225 -0.9466>}
 index_03_r {parent=index_02_r, loc=<-2.5578 0.1231 0.0251>, rot=<-0.002 0.0067 0.2558 -0.9667>}
 middle_metacarpal_r {parent=hand_r, loc=<-2.1856 0.1726 0.5984>, rot=<0.006 0.0168 -0.0329 -0.9993>}
 middle_01_r {parent=middle_metacarpal_r, loc=<-7.3273 0.2447 -0.0129>, rot=<-0.0208 0.0042 0.0858 -0.9961>}
 middle_02_r {parent=middle_01_r, loc=<-4.0142 -0.1411 -0.07>, rot=<-0.0217 0.0259 0.3512 -0.9357>}
 middle_03_r {parent=middle_02_r, loc=<-2.6074 -0.0534 -0.1312>, rot=<0.0083 0.0188 0.1708 -0.9851>}
 pinky_metacarpal_r {parent=hand_r, loc=<-1.7617 0.9753 -3.3814>, rot=<-0.0397 0.0814 -0.069 -0.9935>}
 pinky_01_r {parent=pinky_metacarpal_r, loc=<-7.1707 0.4832 0.1094>, rot=<-0.0654 -0.0035 0.2242 -0.9723>}
 pinky_02_r {parent=pinky_01_r, loc=<-2.8274 -0.0892 0.1043>, rot=<0.0037 0.0011 0.2854 -0.9584>}
 pinky_03_r {parent=pinky_02_r, loc=<-2.2045 -0.031 0.0078>, rot=<-0.0002 0.0098 0.1445 -0.9895>}
 ring_metacarpal_r {parent=hand_r, loc=<-2.2073 0.3627 -1.3638>, rot=<-0.0284 0.0722 -0.0458 -0.9959>}
 ring_01_r {parent=ring_metacarpal_r, loc=<-7.0297 0.6197 0.1446>, rot=<-0.0489 -0.0252 0.1892 -0.9804>}
 ring_02_r {parent=ring_01_r, loc=<-3.6171 -0.4648 0.1027>, rot=<0 0 0.2806 -0.9598>}
 ring_03_r {parent=ring_02_r, loc=<-2.4153 -0.1141 -0.0199>, rot=<0 0 0.1189 -0.9929>}
 thumb_01_r {parent=hand_r, loc=<-1.3119 0.3192 2.459>, rot=<0.6546 -0.4042 0.0664 -0.6354>}
 thumb_02_r {parent=thumb_01_r, loc=<-5.9335 0.0447 -0.0401>, rot=<-0.0037 0.0213 0.298 -0.9543>}
 thumb_03_r {parent=thumb_02_r, loc=<-2.2223 -0.3682 0.0907>, rot=<0.0118 -0.013 0.0066 -0.9998>}
 weapon_r {parent=hand_r, loc=<0.1975 3.4112 -0.3811>, rot=<0 0 0 -1>}
 lowerarm_twist_01_r {parent=lowerarm_r, loc=<-17.855 0 0>, rot=<0 0 0 1>}
 lowerarm_twist_02_r {parent=lowerarm_r, loc=<-9.6199 0 0>, rot=<0 0 0 1>}
 upperarm_twist_01_r {parent=upperarm_r, loc=<-0.0041 0 0>, rot=<0 0 0 -1>}
 upperarm_twist_02_r {parent=upperarm_r, loc=<-13.7942 0 0>, rot=<0 0 0 -1>}
 dyn_shoulder_armor_LOCALSPACE_r {parent=upperarm_r, loc=<-0.607 0 0>, rot=<0 0 0 -1>}
 dyn_shoulder_armor_driver_r {parent=dyn_shoulder_armor_LOCALSPACE_r, loc=<11.4221 0.4451 -0.1588>, rot=<0.8251 0.014 0.5646 -0.016>}
 dyn_shoulder_armor_r {parent=dyn_shoulder_armor_driver_r, loc=<0 0 0>, rot=<0 0 0 1>}
 neck_01 {parent=spine_05, loc=<11.3138 0.1529 0>, rot=<0 0 -0.2003 -0.9797>}
 neck_02 {parent=neck_01, loc=<4.5627 0 0>, rot=<0 0 0.0504 -0.9987>}
 head {parent=neck_02, loc=<5.3914 0 0>, rot=<0 0 0.115 -0.9934>}
 dyn_helmet_rear01 {parent=head, loc=<4.3432 1.7829 0>, rot=<0.0029 -0.7071 -0.0029 -0.7071>}
 dyn_helmet_rear02 {parent=head, loc=<3.4454 1.7992 0>, rot=<0.0029 -0.7071 -0.0029 -0.7071>}
 dyn_helmet_rear03 {parent=head, loc=<0.8658 2.3142 0>, rot=<0.0029 -0.7071 -0.0029 -0.7071>}
 dyn_jawbone_01 {parent=head, loc=<-0.6062 4.7913 -0.0048>, rot=<0.0041 -1 0 0>}
 dyn_jawbone_02 {parent=dyn_jawbone_01, loc=<0 0 0>, rot=<0 0 0 1>}
 dyn_hair_01 {parent=head, loc=<14.2123 4.1721 -3.6492>, rot=<0.0029 -0.7071 -0.0029 -0.7071>}
 dyn_hair_03 {parent=head, loc=<7.2242 4.4471 -7.9422>, rot=<0.0041 -1 0 0>}
 dyn_hair_02 {parent=head, loc=<9.2305 -0.1516 -7.662>, rot=<0.1429 -0.9338 0.3246 0.0469>}
 dyn_beardBraid_01_l {parent=head, loc=<-1.1094 8.3143 -3.3157>, rot=<-0.1882 -0.9821 -0.0023 -0.0115>}
 dyn_beardBraid_02_l {parent=dyn_beardBraid_01_l, loc=<10.4667 0 0>, rot=<0 0 0 1>}
 dyn_beardBraid_01_r {parent=head, loc=<-1.1094 8.3143 3.3157>, rot=<-0.1882 -0.9821 0.0023 0.0115>}
 dyn_beardBraid_02_r {parent=dyn_beardBraid_01_r, loc=<10.4667 0 0>, rot=<0 0 0 1>}
 faceAttach {parent=head, loc=<-2.4561 0.754 0.0089>, rot=<0.0082 -0.7071 -0.0082 -0.7071>, scale=<0.89 0.89 0.89>}
 C_brow_mid {parent=faceAttach, loc=<-0.0268 10.3716 8.3415>, rot=<0 -0.0008 0.0002 -1>, scale=<0.4914 0.8 0.6189>}
 L_brow_mid {parent=faceAttach, loc=<2.7906 9.4899 9.1551>, rot=<0.0089 0.027 0.2635 -0.9643>}
 L_brow_outer {parent=faceAttach, loc=<5.2034 6.7077 9.4554>, rot=<0.0761 0.0918 0.4211 -0.8991>}
 L_cheek_inner {parent=faceAttach, loc=<4.1338 7.4964 3.4687>, rot=<-0.0792 -0.2913 0.3152 -0.8997>}
 C_lip_upper_mid {parent=faceAttach, loc=<0.0183 9.7876 1.5565>, rot=<0.2225 0.0006 -0.0002 -0.9749>}
 L_lip_upper_outer {parent=faceAttach, loc=<1.5869 9.2067 1.6233>, rot=<0.035 -0.0754 0.2637 -0.961>}
 L_lip_corner {parent=faceAttach, loc=<2.8126 8.2457 1.1542>, rot=<0.0028 0.0082 -0.0146 -0.9999>}
 C_nose_bridge {parent=faceAttach, loc=<-0.0066 11.3428 5.1258>, rot=<-0.2218 0.0067 0.0019 -0.9751>}
 R_brow_mid {parent=faceAttach, loc=<-2.9364 9.3876 9.2632>, rot=<0.0237 -0.0443 -0.2511 -0.9666>}
 R_brow_outer {parent=faceAttach, loc=<-5.13 6.7548 9.3893>, rot=<0.0311 -0.0352 -0.435 -0.8992>}
 R_cheek_inner {parent=faceAttach, loc=<-4.1052 7.4933 3.4515>, rot=<-0.0863 0.3092 -0.3124 -0.8941>}
 R_lip_corner {parent=faceAttach, loc=<-2.7506 8.2238 1.1996>, rot=<0.0029 -0.008 0.0146 -0.9999>}
 C_jaw {parent=faceAttach, loc=<-0.0041 3.0377 1.2447>, rot=<0 0 0 1>}
 L_lip_lower_outer {parent=C_jaw, loc=<1.5734 6.2963 0.0493>, rot=<0.0139 0.0493 0.3092 -0.9496>}
 C_lip_lower_mid {parent=C_jaw, loc=<0.0148 7.0893 0.0711>, rot=<0 -0.0187 -0.004 -0.9998>}
 R_lip_lower_outer {parent=C_jaw, loc=<-1.56 6.4826 0.068>, rot=<0.0048 -0.0001 -0.2716 -0.9624>}
 teeth_lower {parent=C_jaw, loc=<-0.0969 4.9378 -1.356>, rot=<0 0 0 1>}
 tongue {parent=C_jaw, loc=<0.0077 1.5855 -0.8452>, rot=<-0.1568 0 0 -0.9876>}
 R_eye {parent=faceAttach, loc=<-2.9442 6.7781 8.1686>, rot=<0 0 0.9998 -0.0187>}
 R_eye_lid_upper_mid {parent=R_eye, loc=<-0.0591 -0.4316 -0.1364>, rot=<0.0523 -0.0009 0.0167 -0.9985>}
 R_eye_lid_lower_mid {parent=R_eye, loc=<0.2627 -0.3471 -0.6628>, rot=<0.0329 -0.0005 0.0167 -0.9993>}
 L_eye {parent=faceAttach, loc=<2.9561 6.777 8.1681>, rot=<0 0 0.9998 0.0187>}
 L_eye_lid_upper_mid {parent=L_eye, loc=<0.0654 -0.4225 -0.1602>, rot=<0.0523 0 0 -0.9986>}
 L_eye_lid_lower_mid {parent=L_eye, loc=<-0.2337 -0.3293 -0.7077>, rot=<0.0329 0 0 -0.9995>}
 teeth_upper {parent=faceAttach, loc=<-0.0051 8.0039 2.0491>, rot=<0 0 0 1>}
 R_lip_upper_outer {parent=faceAttach, loc=<-1.5894 9.2487 1.66>, rot=<0.0361 0.0723 -0.2625 -0.9615>}
 dyn_beard_01 {parent=spine_05, loc=<10.6697 11.8295 0>, rot=<0.0284 -0.7065 -0.0284 -0.7065>}
 dyn_chestStrap_05 {parent=spine_05, loc=<0 0 0>, rot=<0.0284 -0.7065 -0.0284 -0.7065>}
 dyn_neck_hair {parent=spine_05, loc=<12.1424 -6.4889 0>, rot=<0.0284 -0.7065 -0.0284 -0.7065>}
 dyn_nude_chestStrap_05 {parent=spine_05, loc=<0 0 0>, rot=<0.0284 -0.7065 -0.0284 -0.7065>}
 dyn_nude_chestStrap_front_05 {parent=spine_05, loc=<0 0 0>, rot=<0.0284 -0.7065 -0.0284 -0.7065>}
 attach_backpack {parent=spine_05, loc=<-0.123 -13.473 0>, rot=<-0.0093 -0.707 0.0093 -0.707>}
 attach_cape {parent=spine_05, loc=<9.017 -11.714 0>, rot=<0.0616 -0.7044 -0.0616 -0.7044>}
 dyn_chestStrap_04 {parent=spine_04, loc=<0 0 0>, rot=<0.0294 -0.7065 -0.0294 -0.7065>}
 dyn_nude_chestStrap_04 {parent=spine_04, loc=<0 0 0>, rot=<0.0294 -0.7065 -0.0294 -0.7065>}
 dyn_nude_chestStrap_front_04 {parent=spine_04, loc=<0 0 0>, rot=<0.0294 -0.7065 -0.0294 -0.7065>}
 dyn_chestStrap_03 {parent=spine_03, loc=<0 0 0>, rot=<0.0387 -0.706 -0.0387 -0.706>}
 dyn_nude_chestStrap_03 {parent=spine_03, loc=<0 0 0>, rot=<0.0387 -0.706 -0.0387 -0.706>}
 dyn_nude_chestStrap_front_03 {parent=spine_03, loc=<0 0 0>, rot=<0.0387 -0.706 -0.0387 -0.706>}
 dyn_chestStrap_02 {parent=spine_02, loc=<0 0 0>, rot=<0.0282 -0.7065 -0.0282 -0.7065>}
 dyn_nude_chestStrap_02 {parent=spine_02, loc=<0 0 0>, rot=<0.0282 -0.7065 -0.0282 -0.7065>}
 dyn_nude_chestStrap_front_02 {parent=spine_02, loc=<0 0 0>, rot=<0.0282 -0.7065 -0.0282 -0.7065>}
 dyn_chestStrap_01 {parent=spine_01, loc=<0 0 0>, rot=<0.0114 -0.707 -0.0114 -0.707>}
 dyn_nude_chestStrap_01 {parent=spine_01, loc=<0 0 0>, rot=<0.0114 -0.707 -0.0114 -0.707>}
 dyn_nude_chestStrap_front_01 {parent=spine_01, loc=<0 0 0>, rot=<0.0114 -0.707 -0.0114 -0.707>}
 thigh_l {parent=pelvis, loc=<-2.5693 -0.0109 -7.7501>, rot=<0.0012 0.0572 0.0155 -0.9982>}
 calf_l {parent=thigh_l, loc=<-39.0485 0 0>, rot=<0.0006 -0.0161 -0.0391 -0.9991>}
 calf_twist_01_l {parent=calf_l, loc=<-35.813 -0.0967 0.0397>, rot=<0.0237 0.0095 -0.0035 -0.9997>}
 calf_twist_02_l {parent=calf_l, loc=<-18.8064 -0.0967 0.0397>, rot=<0.0237 0.0095 -0.0035 -0.9997>}
 dyn_bootfur_l {parent=calf_twist_02_l, loc=<-1.4374 -2.848 -0.4007>, rot=<-0.0498 -0.7417 0.0123 -0.6688>}
 foot_l {parent=calf_l, loc=<-42.561 0.1077 0.4292>, rot=<0.0418 -0.0455 0.0343 -0.9975>}
 ball_l {parent=foot_l, loc=<-6.1598 13.8264 0.6806>, rot=<0.0001 -0.0001 -0.7063 -0.7079>}
 thigh_twist_01_l {parent=thigh_l, loc=<-21.6303 0 0>, rot=<0.0013 0 0 -1>}
 dyn_loincloth_l_driver {parent=thigh_twist_01_l, loc=<25.8236 -0.5973 -7.9903>, rot=<0.0449 -0.6797 -0.0433 -0.7309>}
 dyn_loincloth_l_01 {parent=dyn_loincloth_l_driver, loc=<0 0 0>, rot=<-0.0646 -0.0931 -0.0061 -0.9935>}
 thigh_r {parent=pelvis, loc=<-2.5691 -0.0109 7.7501>, rot=<0.0572 -0.0012 0.9982 0.0155>}
 calf_r {parent=thigh_r, loc=<39.0487 0 0>, rot=<0.0006 -0.0161 -0.0391 -0.9991>}
 calf_twist_01_r {parent=calf_r, loc=<35.8132 0.0967 -0.0397>, rot=<0.0237 0.0095 -0.0035 -0.9997>}
 calf_twist_02_r {parent=calf_r, loc=<18.8064 0.0967 -0.0397>, rot=<0.0237 0.0095 -0.0035 -0.9997>}
 dyn_bootfur_r {parent=calf_twist_02_r, loc=<1.4373 2.848 0.4007>, rot=<-0.6688 0.0123 0.7417 0.0498>}
 foot_r {parent=calf_r, loc=<42.5612 -0.1077 -0.4292>, rot=<0.0418 -0.0455 0.0343 -0.9975>}
 ball_r {parent=foot_r, loc=<6.1598 -13.8264 -0.6806>, rot=<0.0001 -0.0001 -0.7063 -0.7079>}
 thigh_twist_01_r {parent=thigh_r, loc=<21.6303 0 0>, rot=<0.0013 0 0 -1>}
 dyn_loincloth_r_driver {parent=thigh_twist_01_r, loc=<-25.8235 0.5973 7.9898>, rot=<0.7293 0.0576 -0.6813 0.0256>}
 dyn_loincloth_r_01 {parent=dyn_loincloth_r_driver, loc=<0 0 0>, rot=<-0.0624 0.0908 -0.0178 -0.9938>}
 dyn_horn_01 {parent=pelvis, loc=<6.0133 -5.9041 9.8187>, rot=<-0.021 -0.8829 0.4681 0.0296>}
 dyn_horn_02 {parent=dyn_horn_01, loc=<0 0 0>, rot=<0 0 0 1>}
 dyn_belt_buckle {parent=pelvis, loc=<2.6772 13.2066 -0.3301>, rot=<-0.0127 -0.707 0.0127 -0.707>}
 dyn_rear_loincloth_driver {parent=pelvis, loc=<4.1333 -9.633 -0.3727>, rot=<0 0 0 1>}
 dyn_rear_loincloth_01 {parent=dyn_rear_loincloth_driver, loc=<0 0 0>, rot=<0.0711 -0.9975 0 0>}
 dyn_rear_loincloth_02 {parent=dyn_rear_loincloth_01, loc=<11.9216 0 0>, rot=<0 0 -0.0142 -0.9999>}
 dyn_rear_loincloth_03 {parent=dyn_rear_loincloth_02, loc=<10.9003 0 0>, rot=<0 0 -0.0375 -0.9993>}
 dyn_front_loincloth_PARENT {parent=pelvis, loc=<3.0083 11.4078 0.4572>, rot=<-0.0127 -0.707 0.0127 -0.707>}
 dyn_front_loincloth_LOCAL {parent=dyn_front_loincloth_PARENT, loc=<0 0 0>, rot=<0 0 0 1>}
 dyn_front_loincloth_driver {parent=dyn_front_loincloth_LOCAL, loc=<0 0 0>, rot=<0 0 0 1>}
 dyn_front_loincloth {parent=dyn_front_loincloth_driver, loc=<0 0 0>, rot=<0 0 0 1>}
 dyn_front_loincloth_02 {parent=dyn_front_loincloth, loc=<0 2.4828 -13.088>, rot=<0 0 0 1>}
 dyn_front_loincloth_03 {parent=dyn_front_loincloth_02, loc=<0 1.3322 -9.8013>, rot=<0 0 0 1>}
 front_loin_MASTER {parent=pelvis, loc=<2.2718 10.4373 0.2011>, rot=<-0.1064 -0.9943 0.0001 0.0015>}
 front_loin_Driven_r {parent=front_loin_MASTER, loc=<1.2351 -2.7421 -9.1301>, rot=<-0.0111 -0.0736 0.0424 -0.9963>}
 front_loin_01_r {parent=front_loin_Driven_r, loc=<0 0 0>, rot=<0 0 0 1>}
 front_loin_02_r {parent=front_loin_01_r, loc=<12.9324 0 0>, rot=<0 0 0 1>}
 front_loin_03_r {parent=front_loin_02_r, loc=<21.4342 0 0>, rot=<0.0288 0.7557 0.0249 -0.6538>}
 front_loin_Driven_l {parent=front_loin_MASTER, loc=<0.21 -2.2799 8.8899>, rot=<0.0065 0.025 0.0134 -0.9996>}
 front_loin_01_l {parent=front_loin_Driven_l, loc=<0 0 0>, rot=<0 0 0 1>}
 front_loin_02_l {parent=front_loin_01_l, loc=<15.7212 0 0>, rot=<0 0 0 1>}
 front_loin_03_l {parent=front_loin_02_l, loc=<14.0773 0 0>, rot=<0.0615 0.697 0.0628 -0.7117>}
 ik_foot_root {parent=root, loc=<0 0 0>, rot=<0 0 0 1>}
 ik_foot_l {parent=ik_foot_root, loc=<15.2608 -4.4176 9.4433>, rot=<0.0357 0.7045 0.0284 -0.7082>}
 ik_foot_r {parent=ik_foot_root, loc=<-15.2609 -4.4176 9.4431>, rot=<0.7082 0.0284 -0.7045 0.0357>}
 ik_hand_root {parent=root, loc=<0 0 0>, rot=<0 0 0 1>}
 ik_hand_gun {parent=ik_hand_root, loc=<-33.8029 3.8283 90.5694>, rot=<0.4834 0.3223 0.5887 -0.5621>}
 ik_hand_l {parent=ik_hand_gun, loc=<6.7067 65.8041 13.978>, rot=<0 0 0 1>}
 ik_hand_r {parent=ik_hand_gun, loc=<0 0 0>, rot=<0 0 0 1>}
 Pose_0 {type=Curve, loc=<0 0 0>, rot=<0 0 0 1>}
 Pose_1 {type=Curve, loc=<0 0 0>, rot=<0 0 0 1>}
 Pose_2 {type=Curve, loc=<0 0 0>, rot=<0 0 0 1>}
 Pose_3 {type=Curve, loc=<0 0 0>, rot=<0 0 0 1>}
 Pose_4 {type=Curve, loc=<0 0 0>, rot=<0 0 0 1>}
 Pose_5 {type=Curve, loc=<0 0 0>, rot=<0 0 0 1>}
 Pose_6 {type=Curve, loc=<0 0 0>, rot=<0 0 0 1>}
 Pose_7 {type=Curve, loc=<0 0 0>, rot=<0 0 0 1>}
 Body_T123456_BeltMorph {type=Curve, loc=<0 0 0>, rot=<0 0 0 1>}
 Body_T123456_ArmorMorph {type=Curve, loc=<0 0 0>, rot=<0 0 0 1>}
 Pose_8 {type=Curve, loc=<0 0 0>, rot=<0 0 0 1>}
 Pose_9 {type=Curve, loc=<0 0 0>, rot=<0 0 0 1>}
 Pose_10 {type=Curve, loc=<0 0 0>, rot=<0 0 0 1>}
 Pose_14 {type=Curve, loc=<0 0 0>, rot=<0 0 0 1>}
 Pose_15 {type=Curve, loc=<0 0 0>, rot=<0 0 0 1>}
 Pose_16 {type=Curve, loc=<0 0 0>, rot=<0 0 0 1>}
 Pose_17 {type=Curve, loc=<0 0 0>, rot=<0 0 0 1>}
 Pose_18 {type=Curve, loc=<0 0 0>, rot=<0 0 0 1>}
 Pose_19 {type=Curve, loc=<0 0 0>, rot=<0 0 0 1>}
 Pose_20 {type=Curve, loc=<0 0 0>, rot=<0 0 0 1>}
 Pose_21 {type=Curve, loc=<0 0 0>, rot=<0 0 0 1>}
 Pose_11 {type=Curve, loc=<0 0 0>, rot=<0 0 0 1>}
 Pose_12 {type=Curve, loc=<0 0 0>, rot=<0 0 0 1>}
 Pose_13 {type=Curve, loc=<0 0 0>, rot=<0 0 0 1>}
 Pose_22 {type=Curve, loc=<0 0 0>, rot=<0 0 0 1>}
]
