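# Physics asset for the Tier 4 Dark Viking medium male player body, covering the lower body and a rear loincloth.
# Capsules on the pelvis, thighs and calves sit alongside a three-bone loincloth chain hanging from a tiny driver body linked to the pelvis.
# Cloth links have moderate swing and tight twist limits, while thigh and driver joints to the pelvis leave all angular motion free.
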PhysicsAsset /Game/Characters/Player/Male/Medium/Bodies/M_MED_DarkViking/Meshes/M_MED_DarkViking_01_Tier_4_Physics
bodies {
 calf_l [{shapeType=Capsule, length=20.9909, radius=11.9915, center=<-20.9042 -0.6449 0.4806>, rotation=<rot 0 98.5477 -89.1693>}]
 calf_r [{shapeType=Capsule, length=20.9768, radius=11.6677, center=<20.2014 0.714 -0.4909>, rotation=<rot 0 102.3901 -89.2585>}]
 dyn_rear_loincloth_01 [{shapeType=Capsule, length=12, radius=5, shapeName=dyn_rear_loincloth_01_capsule, center=<6 -5 0>}]
 dyn_rear_loincloth_02 [{shapeType=Capsule, length=18, radius=6, center=<5 -5 0>}]
 dyn_rear_loincloth_03 [{shapeType=Capsule, length=18, radius=8, shapeName=dyn_rear_loincloth_03_capsule, center=<5 -5 0>}]
 dyn_rear_loincloth_driver [{shapeType=Capsule, length=0.505, radius=0.505, shapeName=dyn_rear_loincloth_driver_capsule}]
 pelvis [{shapeType=Capsule, length=9.6824, radius=9.9758, center=<-5.5278 -1.082 0>}]
 thigh_l [{shapeType=Capsule, length=25.0254, radius=8.924, shapeName=thigh_l_capsule, center=<-20 0 0>, rotation=<rot 90 0 0>}]
 thigh_r [{shapeType=Capsule, length=25.0254, radius=8.924, shapeName=thigh_r_capsule, center=<20 0 0>, rotation=<rot 90 0 0>}]
}
constraints [
 {bone1Name=calf_l, bone2Name=thigh_l, swing1LimitDegrees=45, swing2LimitDegrees=45, twistLimitDegrees=45}
 {bone1Name=calf_r, bone2Name=thigh_r, swing1LimitDegrees=45, swing2LimitDegrees=45, twistLimitDegrees=45}
 {bone1Name=dyn_rear_loincloth_02, bone2Name=dyn_rear_loincloth_01, swing1LimitDegrees=35, swing2LimitDegrees=35, twistLimitDegrees=10}
 {
  bone1Name=thigh_l
  bone2Name=pelvis
  swing1LimitDegrees=45
  swing1Motion=Free
  swing2LimitDegrees=45
  swing2Motion=Free
  twistLimitDegrees=45
  twistMotion=Free
 }
 {
  bone1Name=thigh_r
  bone2Name=pelvis
  swing1LimitDegrees=45
  swing1Motion=Free
  swing2LimitDegrees=45
  swing2Motion=Free
  twistLimitDegrees=45
  twistMotion=Free
 }
 {bone1Name=dyn_rear_loincloth_03, bone2Name=dyn_rear_loincloth_02, swing1LimitDegrees=35, swing2LimitDegrees=35, twistLimitDegrees=10}
 {bone1Name=dyn_rear_loincloth_01, bone2Name=dyn_rear_loincloth_driver, swing1LimitDegrees=35, swing2LimitDegrees=35, twistLimitDegrees=10}
 {
  bone1Name=dyn_rear_loincloth_driver
  bone2Name=pelvis
  swing1LimitDegrees=45
  swing1Motion=Free
  swing2LimitDegrees=45
  swing2Motion=Free
  twistLimitDegrees=45
  twistMotion=Free
 }
]
