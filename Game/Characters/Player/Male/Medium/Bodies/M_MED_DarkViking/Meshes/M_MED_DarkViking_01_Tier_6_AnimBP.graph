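# Post-process style AnimBlueprint for the Tier 6 Dark Viking player body, animating its armor and accessory bones over the parent mesh's pose.
# Copy Pose From Mesh feeds a reused Tier 4 Control Rig, whose cached result drives RBF Pose Drivers for chest straps and shoulder armor, with a skydiving variant.
# Bone-driven controllers, Look Ats and thigh-limited AnimDynamics move the loincloth.
# Further dynamics handle horn and boot fur.
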AnimBlueprint /Game/Characters/Player/Male/Medium/Bodies/M_MED_DarkViking/Meshes/M_MED_DarkViking_01_Tier_6_AnimBP
target_skeleton=&/M_MED_DarkViking_01_Tier_6_Skeleton
graphs [
 AnimGraph {nodes [
   n1 = Anim Dynamics - Bone: front_loin_01_r(ComponentPose=Local To Component(LocalPose=Use cached pose 'Base_frontskirt_dynamic'()), Alpha=1, ExternalForce="0,0,0") {
    angularBiasOverride=0.999
    angularDampingOverride=0.999
    angularSpringConstant=300
    bAngularSpring=true
    bGravityOverrideInSimSpace=true
    bOverrideAngularBias=true
    bOverrideAngularDamping=true
    bUseGravityOverride=true
    bUsePlanarLimit=false
    bUseSphericalLimits=true
    simulationSpace=World
    alphaBoolBlend {blendOption=Linear}
    boundBone {boneName=front_loin_01_r}
    componentAppliedLinearAccClamp=<600 600 600>
    componentLinearAccScale=<0.1 0.1 0.1>
    componentLinearVelScale=<0.1 0.1 0.1>
    gravityOverride {z=-980}
    physicsBodyDefinitions [{
      collisionType=CustomSphere
      sphereCollisionRadius=1
      boundBone {boneName=front_loin_01_r}
      boxExtents=<8 8 5>
      constraintSetup {angularLimitsMax=<0 0 15>, angularLimitsMin=<0 0 0>, angularTarget=<0 0 0>}
      localJointOffset=<0.1 0 0>
     }]
    sphericalLimits [{limitRadius=8, drivingBone {boneName=thigh_r}, sphereLocalOffset=<3 0 0>}]
   }
   n2 = Anim Dynamics - Bone: front_loin_01_l(ComponentPose=@n1, Alpha=1, ExternalForce="0,0,0") {
    angularBiasOverride=0.999
    angularDampingOverride=0.999
    angularSpringConstant=300
    bAngularSpring=true
    bGravityOverrideInSimSpace=true
    bOverrideAngularBias=true
    bOverrideAngularDamping=true
    bUseGravityOverride=true
    bUsePlanarLimit=false
    bUseSphericalLimits=true
    simulationSpace=World
    alphaBoolBlend {blendOption=Linear}
    boundBone {boneName=front_loin_01_l}
    componentAppliedLinearAccClamp=<600 600 600>
    componentLinearAccScale=<0.1 0.1 0.1>
    componentLinearVelScale=<0.1 0.1 0.1>
    gravityOverride {z=-980}
    physicsBodyDefinitions [{
      collisionType=CustomSphere
      sphereCollisionRadius=1
      boundBone {boneName=front_loin_01_l}
      boxExtents=<8 8 5>
      constraintSetup {angularLimitsMax=<0 0 15>, angularLimitsMin=<0 0 3>, angularTarget=<0 0 0>}
      localJointOffset=<0.1 0 0>
     }]
    sphericalLimits [{limitRadius=8, drivingBone {boneName=thigh_l}, sphereLocalOffset=<-3 0 0>}]
   }
   n3 = Anim Dynamics - Bone: front_loin_02_r(ComponentPose=@n2, Alpha=1, ExternalForce="0,0,0") {
    angularBiasOverride=0.999
    angularDampingOverride=0.999
    angularSpringConstant=300
    bAngularSpring=true
    bGravityOverrideInSimSpace=true
    bOverrideAngularBias=true
    bOverrideAngularDamping=true
    bUseGravityOverride=true
    bUsePlanarLimit=false
    bUseSphericalLimits=true
    simulationSpace=World
    alphaBoolBlend {blendOption=Linear}
    boundBone {boneName=front_loin_02_r}
    componentAppliedLinearAccClamp=<300 300 300>
    componentLinearAccScale=<0.1 0.1 0.1>
    componentLinearVelScale=<0.1 0.1 0.1>
    gravityOverride {z=-980}
    physicsBodyDefinitions [{
      collisionType=CustomSphere
      sphereCollisionRadius=1
      boundBone {boneName=front_loin_02_r}
      boxExtents=<8 8 8>
      constraintSetup {angularLimitsMax=<0 0 15>, angularLimitsMin=<0 0 0>, angularTarget=<0 0 0>}
      localJointOffset=<0.1 0 0>
     }]
    sphericalLimits [{limitRadius=11.313, drivingBone {boneName=thigh_r}, sphereLocalOffset=<25.5016 0 0>}]
   }
   n4 = Anim Dynamics - Bone: front_loin_02_l(ComponentPose=@n3, Alpha=1, ExternalForce="0,0,0") {
    angularBiasOverride=0.999
    angularDampingOverride=0.999
    angularSpringConstant=300
    bAngularSpring=true
    bGravityOverrideInSimSpace=true
    bOverrideAngularBias=true
    bOverrideAngularDamping=true
    bUseGravityOverride=true
    bUsePlanarLimit=false
    bUseSphericalLimits=true
    simulationSpace=World
    alphaBoolBlend {blendOption=Linear}
    boundBone {boneName=front_loin_02_l}
    componentAppliedLinearAccClamp=<300 300 300>
    componentLinearAccScale=<0.1 0.1 0.1>
    componentLinearVelScale=<0.1 0.1 0.1>
    gravityOverride {z=-980}
    physicsBodyDefinitions [{
      collisionType=CustomSphere
      sphereCollisionRadius=1
      boundBone {boneName=front_loin_02_l}
      boxExtents=<8 8 8>
      constraintSetup {angularLimitsMax=<0 0 15>, angularLimitsMin=<0 0 3>, angularTarget=<0 0 0>}
      localJointOffset=<0.1 0 0>
     }]
    sphericalLimits [{limitRadius=10.3841, drivingBone {boneName=thigh_l}, sphereLocalOffset=<-21.0473 1.2748 1.1415>}]
   }
   Output Pose(Result=Blend Poses by bool(BlendPose_0=Component To Local(ComponentPose=@n4), BlendPose_1=Use cached pose 'Base_pose'()) {blendType=Linear}) {layerGroup=None}
   n5 = Control Rig(Source=Copy Pose From Mesh() {bUseAttachedParent=true}) {controlRigAssetReference {blueprintRigClass=&/M_MED_DarkViking_01_Tier_4_CtrlRig.M_MED_DarkViking_01_Tier_4_CtrlRig_C}}
   Save cached pose 'copypose_cache'(Pose=@n5) {cacheName=copypose_cache}
   n6 = "front_loin_01_l.rotateZ = thigh_l.rotateZ * -1 - Bone Driven Controller"!(ComponentPose:Component Space Pose Link Structure=Local To Component(LocalPose=Use cached pose 'Base_frontskirt_dynamic'()), Alpha:float=1) -> (Pose:Component Space Pose Link Structure) {
    is=title AnimGraphNode_BoneDrivenController
    bAffectTargetRotationZ=true
    multiplier=-1
    sourceComponent=RotationZ
    alphaBoolBlend {blendOption=Linear}
    sourceBone {boneName=thigh_l}
    targetBone {boneName=front_loin_01_l}
   }
   n7 = "front_loin_01_r.rotateZ = thigh_r.rotateZ * -1 - Bone Driven Controller"!(ComponentPose:Component Space Pose Link Structure=@n6, Alpha:float=1) -> (Pose:Component Space Pose Link Structure) {
    is=title AnimGraphNode_BoneDrivenController
    bAffectTargetRotationZ=true
    multiplier=-1
    sourceComponent=RotationZ
    alphaBoolBlend {blendOption=Linear}
    sourceBone {boneName=thigh_r}
    targetBone {boneName=front_loin_01_r}
   }
   n8 = Look At - Bone: front_loin_02_r(ComponentPose=@n7, Alpha=1) {
    bUseLookUpAxis=true
    alphaBoolBlend {blendOption=Linear}
    boneToModify {boneName=front_loin_02_r}
    lookAtLocation {x=0}
    lookAtTarget {boneReference {boneName=front_loin_02_l}}
    lookAt_Axis {axis {y=0, z=1}}
   }
   n9 = Look At - Bone: front_loin_02_l(ComponentPose=@n8, Alpha=1) {
    alphaBoolBlend {blendOption=Linear}
    boneToModify {boneName=front_loin_02_l}
    lookAtLocation {x=0}
    lookAtTarget {boneReference {boneName=front_loin_02_r}}
    lookAt_Axis {axis {y=0, z=-1}}
    lookUp_Axis {axis=<0 1>}
   }
   n10 = Anim Dynamics - Bone: front_loin_01_r(ComponentPose=@n9, Alpha=1, ExternalForce="0,0,0") {
    angularBiasOverride=0.9
    angularDampingOverride=0.995
    angularSpringConstant=600
    bGravityOverrideInSimSpace=true
    bOverrideAngularBias=true
    bOverrideAngularDamping=true
    bUsePlanarLimit=false
    bUseSphericalLimits=true
    alphaBoolBlend {blendOption=Linear}
    boundBone {boneName=front_loin_01_r}
    componentAppliedLinearAccClamp=<4500 4500 4500>
    physicsBodyDefinitions [{
      collisionType=CustomSphere
      sphereCollisionRadius=1
      boundBone {boneName=front_loin_01_r}
      boxExtents=<15 15 15>
      constraintSetup {angularLimitsMax=<0 0 40>, angularLimitsMin=<0 0 -20>, angularTarget=<0 0 0>}
      localJointOffset=<5 0 0>
     }]
    sphericalLimits [{limitRadius=8, drivingBone {boneName=thigh_r}, sphereLocalOffset=<3 0 0>}]
   }
   n11 = Anim Dynamics - Bone: front_loin_01_l(ComponentPose=@n10, Alpha=1, ExternalForce="0,0,0") {
    angularBiasOverride=0.9
    angularDampingOverride=0.995
    angularSpringConstant=600
    bGravityOverrideInSimSpace=true
    bOverrideAngularBias=true
    bOverrideAngularDamping=true
    bUsePlanarLimit=false
    bUseSphericalLimits=true
    alphaBoolBlend {blendOption=Linear}
    boundBone {boneName=front_loin_01_l}
    componentAppliedLinearAccClamp=<4500 4500 4500>
    physicsBodyDefinitions [{
      collisionType=CustomSphere
      sphereCollisionRadius=1
      boundBone {boneName=front_loin_01_l}
      boxExtents=<15 15 15>
      constraintSetup {angularLimitsMax=<0 0 40>, angularLimitsMin=<0 0 -20>, angularTarget=<0 0 0>}
      localJointOffset=<5 0 0>
     }]
    sphericalLimits [{limitRadius=8, drivingBone {boneName=thigh_l}, sphereLocalOffset=<-3 0 0>}]
   }
   n12 = Anim Dynamics - Bone: front_loin_02_r(ComponentPose=@n11, Alpha=1, ExternalForce="0,0,0") {
    angularBiasOverride=0.9
    angularDampingOverride=0.995
    angularSpringConstant=600
    bGravityOverrideInSimSpace=true
    bOverrideAngularBias=true
    bOverrideAngularDamping=true
    bUsePlanarLimit=false
    bUseSphericalLimits=true
    alphaBoolBlend {blendOption=Linear}
    boundBone {boneName=front_loin_02_r}
    componentAppliedLinearAccClamp=<4500 4500 4500>
    physicsBodyDefinitions [{
      collisionType=CustomSphere
      sphereCollisionRadius=1
      boundBone {boneName=front_loin_02_r}
      boxExtents=<15 15 15>
      constraintSetup {angularLimitsMax=<0 0 40>, angularLimitsMin=<0 0 -20>, angularTarget=<0 0 0>}
      localJointOffset=<5 0 0>
     }]
    sphericalLimits [{limitRadius=8, drivingBone {boneName=thigh_r}, sphereLocalOffset=<16 0 0>}]
   }
   n13 = Anim Dynamics - Bone: front_loin_02_l(ComponentPose=@n12, Alpha=1, ExternalForce="0,0,0") {
    angularBiasOverride=0.9
    angularDampingOverride=0.995
    angularSpringConstant=600
    bGravityOverrideInSimSpace=true
    bOverrideAngularBias=true
    bOverrideAngularDamping=true
    bUsePlanarLimit=false
    bUseSphericalLimits=true
    alphaBoolBlend {blendOption=Linear}
    boundBone {boneName=front_loin_02_l}
    componentAppliedLinearAccClamp=<4500 4500 4500>
    physicsBodyDefinitions [{
      collisionType=CustomSphere
      sphereCollisionRadius=1
      boundBone {boneName=front_loin_02_l}
      boxExtents=<15 15 15>
      constraintSetup {angularLimitsMax=<0 0 40>, angularLimitsMin=<0 0 -20>, angularTarget=<0 0 0>}
      localJointOffset=<5 0 0>
     }]
    sphericalLimits [{limitRadius=8, drivingBone {boneName=thigh_l}, sphereLocalOffset=<-16 0 0>}]
   }
   n14 = Anim Dynamics - Bone: dyn_loincloth_r_01(ComponentPose=@n13, Alpha=0.2, ExternalForce="0,0,0") {
    bGravityOverrideInSimSpace=true
    gravityScale=0.2
    alphaBoolBlend {blendOption=Linear}
    boundBone {boneName=dyn_loincloth_r_01}
    componentAppliedLinearAccClamp=<4500 4500 4500>
    componentLinearAccScale=<1 1 1>
    componentLinearVelScale=<1 1 1>
    physicsBodyDefinitions [{
      sphereCollisionRadius=0
      boundBone {boneName=dyn_loincloth_r_01}
      boxExtents=<5 5 5>
      constraintSetup {angularTargetAxis=AxisZ, twistAxis=AxisZ, angularLimitsMax=<10 20 0>, angularLimitsMin=<-10 2 0>, angularTarget=<0 0 0>}
      localJointOffset=<0 0 -1>
     }]
   }
   n15 = Anim Dynamics - Bone: dyn_loincloth_l_01(ComponentPose=@n14, Alpha=0.2, ExternalForce="0,0,0") {
    bGravityOverrideInSimSpace=true
    gravityScale=0.2
    alphaBoolBlend {blendOption=Linear}
    boundBone {boneName=dyn_loincloth_l_01}
    componentAppliedLinearAccClamp=<4500 4500 4500>
    componentLinearAccScale=<1 1 1>
    componentLinearVelScale=<1 1 1>
    physicsBodyDefinitions [{
      sphereCollisionRadius=0
      boundBone {boneName=dyn_loincloth_l_01}
      boxExtents=<5 5 5>
      constraintSetup {angularTargetAxis=AxisZ, twistAxis=AxisZ, angularLimitsMax=<10 2 0>, angularLimitsMin=<-10 -20 0>, angularTarget=<0 0 0>}
      localJointOffset=<0 0 -1>
     }]
   }
   n16 = Anim Dynamics - Bone: dyn_horn_01(ComponentPose=@n15, Alpha=1, ExternalForce="0,0,0") {
    angularDampingOverride=0.9999
    angularSpringConstant=75
    bAngularSpring=true
    bGravityOverrideInSimSpace=true
    bOverrideAngularDamping=true
    simulationSpace=BoneRelative
    alphaBoolBlend {blendOption=Linear}
    boundBone {boneName=dyn_horn_01}
    componentAppliedLinearAccClamp=<4500 4500 4500>
    componentLinearAccScale=<1 1 1>
    componentLinearVelScale=<1 1 1>
    physicsBodyDefinitions [{
      sphereCollisionRadius=0
      boundBone {boneName=dyn_horn_01}
      boxExtents=<10 20 20>
      constraintSetup {angularLimitsMax=<15 15 0>, angularLimitsMin=<-15 -15 -50>, angularTarget=<0 0 0>}
      localJointOffset=<2 0 0>
     }]
    relativeSpaceBone {boneName=root}
   }
   n17 = Anim Dynamics - Bone: dyn_bootfur_r(ComponentPose=@n16, Alpha=0.9, ExternalForce="0,0,0") {
    bGravityOverrideInSimSpace=true
    bLinearSpring=true
    bOverrideLinearDamping=true
    gravityScale=0.1
    lODThreshold=1
    linearDampingOverride=0.995
    linearSpringConstant=200
    alphaBoolBlend {blendOption=Linear}
    boundBone {boneName=dyn_bootfur_r}
    componentAppliedLinearAccClamp=<4500 4500 4500>
    componentLinearAccScale=<1 1 1>
    componentLinearVelScale=<1 1 1>
    physicsBodyDefinitions [{
      sphereCollisionRadius=0
      boundBone {boneName=dyn_bootfur_r}
      boxExtents=<2 2 2>
      constraintSetup {angularLimitsMax=<0 0 0>, angularLimitsMin=<0 0 0>, angularTarget=<0 0 0>, linearAxesMax=<0 0 2>, linearAxesMin=<0 0 -2>}
      localJointOffset=<0 0 -1>
     }]
   }
   n18 = Anim Dynamics - Bone: dyn_bootfur_l(ComponentPose=@n17, Alpha=0.9, ExternalForce="0,0,0") {
    bGravityOverrideInSimSpace=true
    bLinearSpring=true
    bOverrideLinearDamping=true
    gravityScale=0.1
    lODThreshold=1
    linearDampingOverride=0.995
    linearSpringConstant=200
    alphaBoolBlend {blendOption=Linear}
    boundBone {boneName=dyn_bootfur_l}
    componentAppliedLinearAccClamp=<4500 4500 4500>
    componentLinearAccScale=<1 1 1>
    componentLinearVelScale=<1 1 1>
    physicsBodyDefinitions [{
      sphereCollisionRadius=0
      boundBone {boneName=dyn_bootfur_l}
      boxExtents=<2 2 2>
      constraintSetup {angularLimitsMax=<0 0 0>, angularLimitsMin=<0 0 0>, angularTarget=<0 0 0>, linearAxesMax=<0 0 2>, linearAxesMin=<0 0 -2>}
      localJointOffset=<0 0 -1>
     }]
   }
   n19 = "Constraint - dyn_shoulder_armor_driver_l (P)"(ComponentPose=@n18, Alpha=1, ConstraintWeights_0=1) {
    alphaBoolBlend {blendOption=Linear}
    boneToModify {boneName=dyn_shoulder_armor_driver_l}
    constraintSetup [{targetBone {boneName=upperarm_twist_01_l}}]
    constraintWeights [1]
   }
   n20 = "Transform (Modify) Bone - Bone: dyn_shoulder_armor_LOCALSPACE_l"(ComponentPose=@n19, Alpha=1, Translation="5,-2,2", Rotation="-5,1.087,-19.131", Scale="1,1,1") {
    rotationMode=BMM_Additive
    rotationSpace=BCS_BoneSpace
    scaleMode=BMM_Additive
    scaleSpace=BCS_BoneSpace
    translationMode=BMM_Additive
    translationSpace=BCS_ParentBoneSpace
    alphaBoolBlend {blendOption=Linear}
    boneToModify {boneName=dyn_shoulder_armor_LOCALSPACE_l}
    rotation=<rot -5 1.087 -19.131>
    translation=<5 -2 2>
   }
   n21 = "Transform (Modify) Bone - Bone: dyn_shoulder_armor_LOCALSPACE_l"!(ComponentPose=@n20, bAlphaBoolEnabled:bool=True, Translation="6,0,0", Rotation="0,0,0", Scale="1,1,1") -> (Pose) {
    alphaInputType=Bool
    bAlphaBoolEnabled=false
    rotationMode=BMM_Additive
    rotationSpace=BCS_BoneSpace
    scaleMode=BMM_Additive
    scaleSpace=BCS_BoneSpace
    translationMode=BMM_Additive
    translationSpace=BCS_ParentBoneSpace
    alphaBoolBlend {blendInTime=1, blendOption=Linear}
    boneToModify {boneName=dyn_shoulder_armor_LOCALSPACE_l}
    translation {x=6}
   }
   Save cached pose 'Base_pose'(Pose=Component To Local(ComponentPose=@n21)) {cacheName=Base_pose}
   n22 = Pose Driver - Source: upperarm_l!(SourcePose:Pose Link Structure=Use cached pose 'copypose_cache'()) -> (Pose:Pose Link Structure) {
    is=title AnimGraphNode_PoseDriver
    poseAsset=&/M_MED_DarkViking_RBF_Shoulders_Skydiving_PoseAsset
    onlyDriveBones [
     {boneName=dyn_shoulder_armor_LOCALSPACE_l}
     {boneName=dyn_shoulder_armor_driver_l}
     {boneName=dyn_shoulder_armor_driver_B_l}
     {boneName=dyn_shoulder_armor_l}
    ]
    poseTargets [
     {
      bApplyCustomCurve=false
      bIsHidden=false
      distanceMethod=DefaultMethod
      drivenName=Pose_11
      functionType=DefaultFunction
      targetScale=1
      boneTransforms [{targetRotation=<rot -64.7854 -3.089 6.0959>, targetTranslation=<15.5111 0 0>}]
      customCurve {defaultValue=3.4028234663852886e+38, postInfinityExtrap=RCCE_Constant, preInfinityExtrap=RCCE_Constant, keys []}
      targetRotation=<rot 0 0 0>
     }
     {
      bApplyCustomCurve=false
      bIsHidden=false
      distanceMethod=DefaultMethod
      drivenName=Pose_12
      functionType=DefaultFunction
      targetScale=0.7385
      boneTransforms [{targetRotation=<rot -17.9906 -27.503 -67.4976>, targetTranslation=<12.5261 -7.9589 -4.501>}]
      customCurve {defaultValue=3.4028234663852886e+38, postInfinityExtrap=RCCE_Constant, preInfinityExtrap=RCCE_Constant, keys []}
      targetRotation=<rot 0 0 0>
     }
     {
      bApplyCustomCurve=false
      bIsHidden=false
      distanceMethod=DefaultMethod
      drivenName=Pose_13
      functionType=DefaultFunction
      targetScale=0.7385
      boneTransforms [{targetRotation=<rot 17.157 -17.077 -68.3926>, targetTranslation=<15.5086 -0.002 -0.0006>}]
      customCurve {defaultValue=3.4028234663852886e+38, postInfinityExtrap=RCCE_Constant, preInfinityExtrap=RCCE_Constant, keys []}
      targetRotation=<rot 0 0 0>
     }
    ]
    rBFParams {distanceMethod=Quaternion, function=Linear, radius=100}
    sourceBones [{boneName=upperarm_l}]
   }
   n23 = Pose Driver - Source: upperarm_r!(SourcePose:Pose Link Structure=@n22) -> (Pose:Pose Link Structure) {
    is=title AnimGraphNode_PoseDriver
    poseAsset=&M_MED_DarkViking_RBF_Shoulders_Skydiving_PoseAsset
    onlyDriveBones [{boneName=dyn_shoulder_armor_LOCALSPACE_r}, {boneName=dyn_shoulder_armor_driver_r}, {boneName=dyn_shoulder_armor_r}]
    poseTargets [
     {
      bApplyCustomCurve=false
      bIsHidden=false
      distanceMethod=DefaultMethod
      drivenName=Pose_11
      functionType=DefaultFunction
      targetScale=1
      boneTransforms [{targetRotation=<rot -64.7854 -3.089 6.0959>, targetTranslation=<15.5111 0 0>}]
      customCurve {defaultValue=3.4028234663852886e+38, postInfinityExtrap=RCCE_Constant, preInfinityExtrap=RCCE_Constant, keys []}
      targetRotation=<rot 0 0 0>
     }
     {
      bApplyCustomCurve=false
      bIsHidden=false
      distanceMethod=DefaultMethod
      drivenName=Pose_12
      functionType=DefaultFunction
      targetScale=0.7385
      boneTransforms [{targetRotation=<rot -17.9906 -27.503 -67.4976>, targetTranslation=<12.5261 -7.9589 -4.501>}]
      customCurve {defaultValue=3.4028234663852886e+38, postInfinityExtrap=RCCE_Constant, preInfinityExtrap=RCCE_Constant, keys []}
      targetRotation=<rot 0 0 0>
     }
     {
      bApplyCustomCurve=false
      bIsHidden=false
      distanceMethod=DefaultMethod
      drivenName=Pose_13
      functionType=DefaultFunction
      targetScale=0.7385
      boneTransforms [{targetRotation=<rot 17.157 -17.077 -68.3926>, targetTranslation=<15.5086 -0.002 -0.0006>}]
      customCurve {defaultValue=3.4028234663852886e+38, postInfinityExtrap=RCCE_Constant, preInfinityExtrap=RCCE_Constant, keys []}
      targetRotation=<rot 0 0 0>
     }
    ]
    rBFParams {distanceMethod=Quaternion, function=Linear, radius=100}
    sourceBones [{boneName=upperarm_r}]
   }
   n24 = Pose Driver - Source: clavicle_l!(SourcePose:Pose Link Structure=Use cached pose 'copypose_cache'()) -> (Pose:Pose Link Structure) {
    is=title AnimGraphNode_PoseDriver
    poseAsset=&/M_MED_DarkViking_RBF_Shoulders_TakeTwo_PoseAsset
    onlyDriveBones [
     {boneName=dyn_shoulder_armor_LOCALSPACE_l}
     {boneName=dyn_shoulder_armor_driver_l}
     {boneName=dyn_shoulder_armor_driver_B_l}
     {boneName=dyn_shoulder_armor_l}
    ]
    poseTargets [
     {
      bApplyCustomCurve=false
      bIsHidden=false
      distanceMethod=DefaultMethod
      drivenName=Pose_14
      functionType=DefaultFunction
      targetScale=0.2379
      boneTransforms [
       {targetRotation=<rot -85.1913 -138.5624 143.0618>, targetTranslation=<4.3575 -0.1466 -2.3223>}
       {targetRotation=<rot -64.7854 -3.089 6.0959>, targetTranslation=<15.5111 0 0>}
      ]
      customCurve {defaultValue=3.4028234663852886e+38, postInfinityExtrap=RCCE_Constant, preInfinityExtrap=RCCE_Constant, keys []}
      targetRotation=<rot 0 0 0>
     }
     {
      bApplyCustomCurve=false
      bIsHidden=false
      distanceMethod=DefaultMethod
      drivenName=Pose_15
      functionType=DefaultFunction
      targetScale=1
      boneTransforms [
       {targetRotation=<rot -71.8074 -7.0773 -33.3058>, targetTranslation=<4.3575 -0.1466 -2.3223>}
       {targetRotation=<rot -9.5233 -21.9053 -28.615>, targetTranslation=<15.5111 0 0>}
      ]
      customCurve {defaultValue=3.4028234663852886e+38, postInfinityExtrap=RCCE_Constant, preInfinityExtrap=RCCE_Constant, keys []}
      targetRotation=<rot 0 0 0>
     }
     {
      bApplyCustomCurve=false
      bIsHidden=false
      distanceMethod=DefaultMethod
      drivenName=Pose_16
      functionType=DefaultFunction
      targetScale=0.308
      boneTransforms [
       {targetRotation=<rot -75.4617 19.5667 -12.8774>, targetTranslation=<4.3575 -0.1466 -2.3223>}
       {targetRotation=<rot -64.9089 79.0369 -47.2079>, targetTranslation=<15.5111 0 0>}
      ]
      customCurve {defaultValue=3.4028234663852886e+38, postInfinityExtrap=RCCE_Constant, preInfinityExtrap=RCCE_Constant, keys []}
      targetRotation=<rot 0 0 0>
     }
     {
      bApplyCustomCurve=false
      bIsHidden=false
      distanceMethod=DefaultMethod
      drivenName=Pose_17
      functionType=DefaultFunction
      targetScale=0.4931
      boneTransforms [
       {targetRotation=<rot -74.7073 55.0349 -49.5566>, targetTranslation=<4.3575 -0.1466 -2.3223>}
       {targetRotation=<rot -36.5188 -9.2495 27.844>, targetTranslation=<15.5111 0 0>}
      ]
      customCurve {defaultValue=3.4028234663852886e+38, postInfinityExtrap=RCCE_Constant, preInfinityExtrap=RCCE_Constant, keys []}
      targetRotation=<rot 0 0 0>
     }
     {
      bApplyCustomCurve=false
      bIsHidden=false
      distanceMethod=DefaultMethod
      drivenName=Pose_18
      functionType=DefaultFunction
      targetScale=0.2379
      boneTransforms [
       {targetRotation=<rot -81.7412 -135.4937 140.2621>, targetTranslation=<4.3575 -0.1466 -2.3223>}
       {targetRotation=<rot -49.8941 -8.1717 1.1726>, targetTranslation=<15.5111 0 0>}
      ]
      customCurve {defaultValue=3.4028234663852886e+38, postInfinityExtrap=RCCE_Constant, preInfinityExtrap=RCCE_Constant, keys []}
      targetRotation=<rot 0 0 0>
     }
     {
      bApplyCustomCurve=false
      bIsHidden=false
      distanceMethod=DefaultMethod
      drivenName=Pose_19
      functionType=DefaultFunction
      targetScale=0.7376
      boneTransforms [
       {targetRotation=<rot -80.8623 39.0398 -38.1228>, targetTranslation=<4.3575 -0.1466 -2.3223>}
       {targetRotation=<rot -31.8256 65.8251 -86.0604>, targetTranslation=<15.5111 0 0>}
      ]
      customCurve {defaultValue=3.4028234663852886e+38, postInfinityExtrap=RCCE_Constant, preInfinityExtrap=RCCE_Constant, keys []}
      targetRotation=<rot 0 0 0>
     }
     {
      bApplyCustomCurve=false
      bIsHidden=false
      distanceMethod=DefaultMethod
      drivenName=Pose_20
      functionType=DefaultFunction
      targetScale=0.558
      boneTransforms [
       {targetRotation=<rot -47.5697 14.692 -0.434>, targetTranslation=<4.3575 -0.1466 -2.3223>}
       {targetRotation=<rot -8.3098 26.4205 -78.9654>, targetTranslation=<15.5111 0 0>}
      ]
      customCurve {defaultValue=3.4028234663852886e+38, postInfinityExtrap=RCCE_Constant, preInfinityExtrap=RCCE_Constant, keys []}
      targetRotation=<rot 0 0 0>
     }
     {
      bApplyCustomCurve=false
      bIsHidden=false
      distanceMethod=DefaultMethod
      drivenName=Pose_21
      functionType=DefaultFunction
      targetScale=0.308
      boneTransforms [
       {targetRotation=<rot -86.0332 143.397 -137.7672>, targetTranslation=<4.3575 -0.1466 -2.3223>}
       {targetRotation=<rot -67.1606 51.1714 -32.7303>, targetTranslation=<15.5111 0 0>}
      ]
      customCurve {defaultValue=3.4028234663852886e+38, postInfinityExtrap=RCCE_Constant, preInfinityExtrap=RCCE_Constant, keys []}
      targetRotation=<rot 0 0 0>
     }
     {
      bApplyCustomCurve=false
      bIsHidden=false
      distanceMethod=DefaultMethod
      drivenName=Pose_11
      functionType=DefaultFunction
      targetScale=0.2698
      boneTransforms [
       {targetRotation=<rot -84.0122 51.6418 -47.6753>, targetTranslation=<4.3575 -0.1466 -2.3223>}
       {targetRotation=<rot -74.2196 -36.1046 27.6902>, targetTranslation=<15.5111 0 0>}
      ]
      customCurve {defaultValue=3.4028234663852886e+38, postInfinityExtrap=RCCE_Constant, preInfinityExtrap=RCCE_Constant, keys []}
      targetRotation=<rot 0 0 0>
     }
     {
      bApplyCustomCurve=false
      bIsHidden=false
      distanceMethod=DefaultMethod
      drivenName=Pose_12
      functionType=DefaultFunction
      targetScale=0.558
      boneTransforms [
       {targetRotation=<rot -48.3561 10.1188 -2.4428>, targetTranslation=<4.3575 -0.1466 -2.3223>}
       {targetRotation=<rot 14.5763 54.8781 -98.5188>, targetTranslation=<15.5111 0 0>}
      ]
      customCurve {defaultValue=3.4028234663852886e+38, postInfinityExtrap=RCCE_Constant, preInfinityExtrap=RCCE_Constant, keys []}
      targetRotation=<rot 0 0 0>
     }
     {
      bApplyCustomCurve=false
      bIsHidden=false
      distanceMethod=DefaultMethod
      drivenName=Pose_13
      functionType=DefaultFunction
      targetScale=0.7228
      boneTransforms [
       {targetRotation=<rot -79.2874 99.6397 -110.9321>, targetTranslation=<4.3575 -0.1466 -2.3223>}
       {targetRotation=<rot -14.9047 102.8909 -17.6919>, targetTranslation=<15.5111 0 0>}
      ]
      customCurve {defaultValue=3.4028234663852886e+38, postInfinityExtrap=RCCE_Constant, preInfinityExtrap=RCCE_Constant, keys []}
      targetRotation=<rot 0 0 0>
     }
     {
      bApplyCustomCurve=false
      bIsHidden=false
      distanceMethod=DefaultMethod
      drivenName=Pose_22
      functionType=DefaultFunction
      targetScale=0.7228
      boneTransforms [
       {targetRotation=<rot -58.4846 -9.5373 2.6759>, targetTranslation=<4.3575 -0.1466 -2.3223>}
       {targetRotation=<rot -44.7022 128.4906 -48.0679>, targetTranslation=<15.5111 0 0>}
      ]
      customCurve {defaultValue=3.4028234663852886e+38, postInfinityExtrap=RCCE_Constant, preInfinityExtrap=RCCE_Constant, keys []}
      targetRotation=<rot 0 0 0>
     }
    ]
    rBFParams {distanceMethod=Quaternion, function=Linear, radius=100}
    sourceBones [{boneName=clavicle_l}, {boneName=upperarm_l}]
   }
   n25 = Pose Driver - Source: clavicle_r!(SourcePose:Pose Link Structure=@n24) -> (Pose:Pose Link Structure) {
    is=title AnimGraphNode_PoseDriver
    poseAsset=&M_MED_DarkViking_RBF_Shoulders_TakeTwo_PoseAsset
    onlyDriveBones [{boneName=dyn_shoulder_armor_LOCALSPACE_r}, {boneName=dyn_shoulder_armor_driver_r}, {boneName=dyn_shoulder_armor_r}]
    poseTargets [
     {
      bApplyCustomCurve=false
      bIsHidden=false
      distanceMethod=DefaultMethod
      drivenName=Pose_14
      functionType=DefaultFunction
      targetScale=0.2938
      boneTransforms [
       {targetRotation=<rot -85.1914 41.4375 143.0618>, targetTranslation=<4.3575 -0.1466 2.3223>}
       {targetRotation=<rot -64.7854 -3.089 6.0959>, targetTranslation=<-15.5111 0 0>}
      ]
      customCurve {defaultValue=3.4028234663852886e+38, postInfinityExtrap=RCCE_Constant, preInfinityExtrap=RCCE_Constant, keys []}
      targetRotation=<rot 0 0 0>
     }
     {
      bApplyCustomCurve=false
      bIsHidden=false
      distanceMethod=DefaultMethod
      drivenName=Pose_15
      functionType=DefaultFunction
      targetScale=0.5605
      boneTransforms [
       {targetRotation=<rot -64.2747 -120.6598 -83.241>, targetTranslation=<4.3575 -0.1466 2.3223>}
       {targetRotation=<rot -23.3709 86.7554 -36.3331>, targetTranslation=<-15.5111 0 0>}
      ]
      customCurve {defaultValue=3.4028234663852886e+38, postInfinityExtrap=RCCE_Constant, preInfinityExtrap=RCCE_Constant, keys []}
      targetRotation=<rot 0 0 0>
     }
     {
      bApplyCustomCurve=false
      bIsHidden=false
      distanceMethod=DefaultMethod
      drivenName=Pose_16
      functionType=DefaultFunction
      targetScale=0.6181
      boneTransforms [
       {targetRotation=<rot -69.838 165.5965 22.6769>, targetTranslation=<4.3575 -0.1466 2.3223>}
       {targetRotation=<rot -75.1088 43.0245 -78.2219>, targetTranslation=<-15.5111 0 0>}
      ]
      customCurve {defaultValue=3.4028234663852886e+38, postInfinityExtrap=RCCE_Constant, preInfinityExtrap=RCCE_Constant, keys []}
      targetRotation=<rot 0 0 0>
     }
     {
      bApplyCustomCurve=false
      bIsHidden=false
      distanceMethod=DefaultMethod
      drivenName=Pose_17
      functionType=DefaultFunction
      targetScale=0.3174
      boneTransforms [
       {targetRotation=<rot -73.4134 -176.9703 4.3592>, targetTranslation=<4.3575 -0.1466 2.3223>}
       {targetRotation=<rot -75.5656 105.7208 -93.8077>, targetTranslation=<-15.5111 0 0>}
      ]
      customCurve {defaultValue=3.4028234663852886e+38, postInfinityExtrap=RCCE_Constant, preInfinityExtrap=RCCE_Constant, keys []}
      targetRotation=<rot 0 0 0>
     }
     {
      bApplyCustomCurve=false
      bIsHidden=false
      distanceMethod=DefaultMethod
      drivenName=Pose_18
      functionType=DefaultFunction
      targetScale=0.2938
      boneTransforms [
       {targetRotation=<rot -87.7993 76.2682 109.0758>, targetTranslation=<4.3575 -0.1466 2.3223>}
       {targetRotation=<rot -69.7589 51.1037 -35.2693>, targetTranslation=<-15.5111 0 0>}
      ]
      customCurve {defaultValue=3.4028234663852886e+38, postInfinityExtrap=RCCE_Constant, preInfinityExtrap=RCCE_Constant, keys []}
      targetRotation=<rot 0 0 0>
     }
     {
      bApplyCustomCurve=false
      bIsHidden=false
      distanceMethod=DefaultMethod
      drivenName=Pose_19
      functionType=DefaultFunction
      targetScale=0.2327
      boneTransforms [
       {targetRotation=<rot -76.2754 172.1506 9.5506>, targetTranslation=<4.3575 -0.1466 2.3223>}
       {targetRotation=<rot -45.0185 104.3596 -56.1379>, targetTranslation=<-15.5111 0 0>}
      ]
      customCurve {defaultValue=3.4028234663852886e+38, postInfinityExtrap=RCCE_Constant, preInfinityExtrap=RCCE_Constant, keys []}
      targetRotation=<rot 0 0 0>
     }
     {
      bApplyCustomCurve=false
      bIsHidden=false
      distanceMethod=DefaultMethod
      drivenName=Pose_20
      functionType=DefaultFunction
      targetScale=0.2327
      boneTransforms [
       {targetRotation=<rot -74.8686 -132.7057 -46.9771>, targetTranslation=<4.3575 -0.1466 2.3223>}
       {targetRotation=<rot -39.6825 113.2505 -70.71>, targetTranslation=<-15.5111 0 0>}
      ]
      customCurve {defaultValue=3.4028234663852886e+38, postInfinityExtrap=RCCE_Constant, preInfinityExtrap=RCCE_Constant, keys []}
      targetRotation=<rot 0 0 0>
     }
     {
      bApplyCustomCurve=false
      bIsHidden=false
      distanceMethod=DefaultMethod
      drivenName=Pose_21
      functionType=DefaultFunction
      targetScale=0.4039
      boneTransforms [
       {targetRotation=<rot -57.8419 -161.5812 -18.7948>, targetTranslation=<4.3575 -0.1466 2.3223>}
       {targetRotation=<rot -66.8684 120.4314 -71.1722>, targetTranslation=<-15.5111 0 0>}
      ]
      customCurve {defaultValue=3.4028234663852886e+38, postInfinityExtrap=RCCE_Constant, preInfinityExtrap=RCCE_Constant, keys []}
      targetRotation=<rot 0 0 0>
     }
     {
      bApplyCustomCurve=false
      bIsHidden=false
      distanceMethod=DefaultMethod
      drivenName=Pose_11
      functionType=DefaultFunction
      targetScale=0.6101
      boneTransforms [
       {targetRotation=<rot -53.6465 -152.7178 -29.1383>, targetTranslation=<4.3575 -0.1466 2.3223>}
       {targetRotation=<rot -17.8084 26.8504 -97.253>, targetTranslation=<-15.5111 0 0>}
      ]
      customCurve {defaultValue=3.4028234663852886e+38, postInfinityExtrap=RCCE_Constant, preInfinityExtrap=RCCE_Constant, keys []}
      targetRotation=<rot 0 0 0>
     }
     {
      bApplyCustomCurve=false
      bIsHidden=false
      distanceMethod=DefaultMethod
      drivenName=Pose_12
      functionType=DefaultFunction
      targetScale=1
      boneTransforms [
       {targetRotation=<rot -68.9933 -127.6298 -53.9107>, targetTranslation=<4.3575 -0.1466 2.3223>}
       {targetRotation=<rot 33.2732 89.3676 -89.1674>, targetTranslation=<-15.5111 0 0>}
      ]
      customCurve {defaultValue=3.4028234663852886e+38, postInfinityExtrap=RCCE_Constant, preInfinityExtrap=RCCE_Constant, keys []}
      targetRotation=<rot 0 0 0>
     }
     {
      bApplyCustomCurve=false
      bIsHidden=false
      distanceMethod=DefaultMethod
      drivenName=Pose_13
      functionType=DefaultFunction
      targetScale=0.6101
      boneTransforms [
       {targetRotation=<rot -83.622 69.4344 115.2173>, targetTranslation=<4.3575 -0.1466 2.3223>}
       {targetRotation=<rot -11.4935 30.1498 -76.7725>, targetTranslation=<-15.5111 0 0>}
      ]
      customCurve {defaultValue=3.4028234663852886e+38, postInfinityExtrap=RCCE_Constant, preInfinityExtrap=RCCE_Constant, keys []}
      targetRotation=<rot 0 0 0>
     }
     {
      bApplyCustomCurve=false
      bIsHidden=false
      distanceMethod=DefaultMethod
      drivenName=Pose_22
      functionType=DefaultFunction
      targetScale=0.571
      boneTransforms [
       {targetRotation=<rot -47.9536 -145.5199 -31.7961>, targetTranslation=<4.3575 -0.1466 2.3223>}
       {targetRotation=<rot -20.6591 112.0123 -37.7437>, targetTranslation=<-15.5111 0 0>}
      ]
      customCurve {defaultValue=3.4028234663852886e+38, postInfinityExtrap=RCCE_Constant, preInfinityExtrap=RCCE_Constant, keys []}
      targetRotation=<rot 0 0 0>
     }
    ]
    rBFParams {distanceMethod=Quaternion, radius=100}
    sourceBones [{boneName=clavicle_r}, {boneName=upperarm_r}]
   }
   n26 = Pose Driver - Source: spine_01!(SourcePose:Pose Link Structure=Blend Poses by bool(BlendPose_0=@n23, BlendPose_1=@n25, bActiveValue=Get bIsSkyDiving()) {blendType=Linear}) -> (Pose:Pose Link Structure) {
    is=title AnimGraphNode_PoseDriver
    poseAsset=&/M_MED_DarkViking_RBF_Spine_PoseAsset
    onlyDriveBones [
     {boneName=dyn_chestStrap_01}
     {boneName=dyn_chestStrap_02}
     {boneName=dyn_chestStrap_03}
     {boneName=dyn_chestStrap_04}
     {boneName=dyn_chestStrap_05}
    ]
    poseTargets [
     {
      bApplyCustomCurve=false
      bIsHidden=false
      distanceMethod=DefaultMethod
      drivenName=Pose_5
      functionType=DefaultFunction
      targetScale=0.8804
      boneTransforms [
       {targetRotation=<rot 0 -3.9009 0>, targetTranslation=<6.7494 -0.6388 0>}
       {targetRotation=<rot 0 -2.7227 0>, targetTranslation=<7.2518 -0.035 0>}
       {targetRotation=<rot 0 -1.7047 0>, targetTranslation=<6.0104 0 0>}
       {targetRotation=<rot 0 1.5078 0>, targetTranslation=<8.1024 0 0>}
       {targetRotation=<rot 0 0.1712 0>, targetTranslation=<10.0146 0 0>}
      ]
      customCurve {defaultValue=3.4028234663852886e+38, postInfinityExtrap=RCCE_Constant, preInfinityExtrap=RCCE_Constant, keys []}
      targetRotation=<rot 0 0 0>
     }
     {
      bApplyCustomCurve=false
      bIsHidden=false
      distanceMethod=DefaultMethod
      drivenName=Pose_6
      functionType=DefaultFunction
      targetScale=0.8804
      boneTransforms [
       {targetRotation=<rot -9.8046 9.1979 4.183>, targetTranslation=<6.7494 -0.6388 0>}
       {targetRotation=<rot -0.9946 1.4714 16.5587>, targetTranslation=<7.2518 -0.035 0>}
       {targetRotation=<rot -0.0758 -1.9943 26.929>, targetTranslation=<6.0104 0 0>}
       {targetRotation=<rot -2.3574 -3.9817 35.5043>, targetTranslation=<8.1024 0 0>}
       {targetRotation=<rot -7.8815 -6.5671 16.7378>, targetTranslation=<10.0146 0 0>}
      ]
      customCurve {defaultValue=3.4028234663852886e+38, postInfinityExtrap=RCCE_Constant, preInfinityExtrap=RCCE_Constant, keys []}
      targetRotation=<rot 0 0 0>
     }
     {
      bApplyCustomCurve=false
      bIsHidden=false
      distanceMethod=DefaultMethod
      drivenName=Pose_7
      functionType=DefaultFunction
      targetScale=1
      boneTransforms [
       {targetRotation=<rot -8.1192 8.789 -10.6853>, targetTranslation=<6.7494 -0.6388 0>}
       {targetRotation=<rot 2.2508 2.7301 -13.4454>, targetTranslation=<7.2518 -0.035 0>}
       {targetRotation=<rot 5.3508 1.6592 -18.9838>, targetTranslation=<6.0104 0 0>}
       {targetRotation=<rot 9.7778 3.9022 -23.3152>, targetTranslation=<8.1024 0 0>}
       {targetRotation=<rot 11.9649 2.7821 -12.2104>, targetTranslation=<10.0146 0 0>}
      ]
      customCurve {defaultValue=3.4028234663852886e+38, postInfinityExtrap=RCCE_Constant, preInfinityExtrap=RCCE_Constant, keys []}
      targetRotation=<rot 0 0 0>
     }
    ]
    rBFParams {radius=15}
    sourceBones [{boneName=spine_01}, {boneName=spine_02}, {boneName=spine_03}, {boneName=spine_04}, {boneName=spine_05}]
   }
   n27 = Pose Driver - Source: spine_01(SourcePose=@n26) {
    poseAsset=&/M_MED_DarkViking_RBF_Spine_nude_PoseAsset
    onlyDriveBones [
     {boneName=dyn_nude_chestStrap_front_01}
     {boneName=dyn_nude_chestStrap_01}
     {boneName=dyn_nude_chestStrap_02}
     {boneName=dyn_nude_chestStrap_front_02}
     {boneName=dyn_nude_chestStrap_front_03}
     {boneName=dyn_nude_chestStrap_03}
     {boneName=dyn_nude_chestStrap_front_04}
     {boneName=dyn_nude_chestStrap_04}
     {boneName=dyn_nude_chestStrap_front_05}
     {boneName=dyn_nude_chestStrap_05}
    ]
    poseTargets [
     {
      bApplyCustomCurve=false
      bIsHidden=false
      distanceMethod=DefaultMethod
      drivenName=Pose_8
      functionType=DefaultFunction
      targetScale=0.8804
      boneTransforms [
       {targetRotation=<rot 0 -3.9009 0>, targetTranslation=<6.7494 -0.6388 0>}
       {targetRotation=<rot 0 -2.7227 0>, targetTranslation=<7.2518 -0.035 0>}
       {targetRotation=<rot 0 -1.7047 0>, targetTranslation=<6.0104 0 0>}
       {targetRotation=<rot 0 1.5078 0>, targetTranslation=<8.1024 0 0>}
       {targetRotation=<rot 0 0.1712 0>, targetTranslation=<10.0146 0 0>}
      ]
      customCurve {defaultValue=3.4028234663852886e+38, postInfinityExtrap=RCCE_Constant, preInfinityExtrap=RCCE_Constant, keys []}
      targetRotation=<rot 0 0 0>
     }
     {
      bApplyCustomCurve=false
      bIsHidden=false
      distanceMethod=DefaultMethod
      drivenName=Pose_9
      functionType=DefaultFunction
      targetScale=0.8804
      boneTransforms [
       {targetRotation=<rot -9.8046 9.1979 4.183>, targetTranslation=<6.7494 -0.6388 0>}
       {targetRotation=<rot -0.9946 1.4714 16.5587>, targetTranslation=<7.2518 -0.035 0>}
       {targetRotation=<rot -0.0758 -1.9943 26.929>, targetTranslation=<6.0104 0 0>}
       {targetRotation=<rot -2.3574 -3.9817 35.5043>, targetTranslation=<8.1024 0 0>}
       {targetRotation=<rot -7.8815 -6.5671 16.7378>, targetTranslation=<10.0146 0 0>}
      ]
      customCurve {defaultValue=3.4028234663852886e+38, postInfinityExtrap=RCCE_Constant, preInfinityExtrap=RCCE_Constant, keys []}
      targetRotation=<rot 0 0 0>
     }
     {
      bApplyCustomCurve=false
      bIsHidden=false
      distanceMethod=DefaultMethod
      drivenName=Pose_10
      functionType=DefaultFunction
      targetScale=1
      boneTransforms [
       {targetRotation=<rot -8.1192 8.789 -10.6853>, targetTranslation=<6.7494 -0.6388 0>}
       {targetRotation=<rot 2.2508 2.7301 -13.4454>, targetTranslation=<7.2518 -0.035 0>}
       {targetRotation=<rot 5.3508 1.6592 -18.9838>, targetTranslation=<6.0104 0 0>}
       {targetRotation=<rot 9.7778 3.9022 -23.3152>, targetTranslation=<8.1024 0 0>}
       {targetRotation=<rot 11.9649 2.7821 -12.2104>, targetTranslation=<10.0146 0 0>}
      ]
      customCurve {defaultValue=3.4028234663852886e+38, postInfinityExtrap=RCCE_Constant, preInfinityExtrap=RCCE_Constant, keys []}
      targetRotation=<rot 0 0 0>
     }
    ]
    rBFParams {function=Linear, radius=21}
    sourceBones [{boneName=spine_01}, {boneName=spine_02}, {boneName=spine_03}, {boneName=spine_04}, {boneName=spine_05}]
   }
   n28 = "Transform (Modify) Bone - Bone: dyn_rear_loincloth_02"!(
    ComponentPose=Local To Component(LocalPose=@n27)
    bAlphaBoolEnabled:bool=True
    Translation="0,0,0"
    Rotation="0.000,-20,0.000"
    Scale="1,1,1"
   ) -> (Pose) {
    alphaInputType=Bool
    bAlphaBoolEnabled=false
    rotationMode=BMM_Additive
    rotationSpace=BCS_BoneSpace
    alphaBoolBlend {blendInTime=0.1, blendOption=Linear, blendOutTime=0.1}
    boneToModify {boneName=dyn_rear_loincloth_02}
    rotation {yaw=-20}
   }
   n29 = RigidBody!(ComponentPose=@n28, Alpha, OverrideWorldGravity:Vector=Get GravityOverride(), ExternalForce) -> (Pose) {
    bClampLinearTranslationLimitToRefPose=true
    bOverrideWorldGravity=true
    overridePhysicsAsset=&/M_MED_DarkViking_01_Tier_4_Physics
    simulationSpace=BaseBoneSpace
    alphaBoolBlend {blendOption=Linear}
    alphaScaleBiasClamp {bias=1, scale=-1}
    baseBoneRef {boneName=pelvis}
    componentAppliedLinearAccClamp=<3000 3000 3000>
    componentLinearAccScale=<0.25 0.25 0.25>
    componentLinearVelScale=<0.2 0.2 0.2>
   }
   Save cached pose 'Base_frontskirt_dynamic'(Pose=Component To Local(ComponentPose=@n29)) {cacheName=Base_frontskirt_dynamic}
   float - float(B="To Float (Boolean)"(InBool=Get isWearingCape() -> (isWearingCape:bool))) {is=title K2Node_CallFunction}
  ]}
 EventGraph {nodes [
   Event BlueprintUpdateAnimation()
   > n1 = Gravity Override State Function DEPRECATED(CharacterPartAnimInstance=Self-Reference(), GameTimeInSeconds=GetGameTimeInSeconds(), GravityOverrideParams=Get Gravity Override Params())
   > Set GravityOverride(GravityOverride=@n1)
   n2 = Get bIsPlayingEmote()
   > Set FrontEnd_notEmote(FrontEnd_notEmote:bool=AND Boolean(A=Get bIsInFrontEnd(), B=NOT Boolean(A=@n2))) -> (Output_Get:bool)
   Event OnCharacterCustomizationLoaded()
   n3 = CheckForBackbling()
   > Branch(Condition=@n3)
   > Set isWearingCape(isWearingCape=@n3.bIsBackblingLong)
  ]}
]
variables [
 FrontEnd_notEmote {type=boolean}
 Gravity Override Params {
  type=GravityOverrideParamsStruct
  default {
   skydive_FloatingDown {gravityOverride=<0 -700 -540>, windAmplitude=<100 -250 200>, windFrequency=<0.75 3.5 4>}
   skydive_GoingUp {gravityOverride=<0 -300 -540>, windAmplitude=<100 -250 200>, windFrequency=<0.75 3.5 4>}
   skydive_Parachuting {windAmplitude=<25 -50 40>, windFrequency=<0.25 1 1.5>}
  }
 }
 GravityOverride {type=Vector}
 isWearingCape {type=boolean}
]
