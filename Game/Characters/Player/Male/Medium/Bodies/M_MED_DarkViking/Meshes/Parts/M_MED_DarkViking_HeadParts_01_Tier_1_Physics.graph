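# Physics asset for the Tier 1 head parts of the male medium Dark Viking player body, covering the beard braids and upper body.
# Capsule bodies form three braid chains from the jaw: a three-bone center braid and two-bone left and right braids.
# Braid joints carry angular limits, except the free center root at the jaw.
# Larger spine and clavicle capsules approximate the torso, with head, clavicle and spine links left angularly free.
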PhysicsAsset /Game/Characters/Player/Male/Medium/Bodies/M_MED_DarkViking/Meshes/Parts/M_MED_DarkViking_HeadParts_01_Tier_1_Physics
bodies {
 C_jaw [{shapeType=Capsule, length=0.505, radius=0.505}]
 clavicle_l [{shapeType=Capsule, length=2.5, radius=6.255, shapeName=clavicle_l_capsule, rotation=<rot 0 -90 0>}]
 clavicle_r [{shapeType=Capsule, length=2.5, radius=6.255, shapeName=clavicle_r_capsule, rotation=<rot 0 -90 0>}]
 dyn_beardBraid_01_c [{shapeType=Capsule, length=0.505, radius=0.505, shapeName=dyn_beardBraid_01_c_capsule}]
 dyn_beardBraid_02_a_c [{shapeType=Capsule, length=2.1891, radius=1.8319, shapeName=dyn_beardBraid_02_a_c_capsule, center=<3.4116 0 0>, rotation=<rot 90 0 0>}]
 dyn_beardBraid_02_b_c [{shapeType=Capsule, length=5.6937, radius=2.2249, shapeName=dyn_beardBraid_02_b_c_capsule, center=<5.1273 0 0>, rotation=<rot 90 0 0>}]
 dyn_beardBraid_T3_01_l [{
   shapeType=Capsule
   length=2.1908
   radius=1.4861
   shapeName=dyn_beardBraid_T3_01_l_capsule
   center=<0 1 3.885>
   rotation=<rot 9.9375 -90 -0.6412>
  }]
 dyn_beardBraid_T3_01_r [{
   shapeType=Capsule
   length=2.1908
   radius=1.5946
   shapeName=dyn_beardBraid_T3_01_r_capsule
   center=<0 1 3.7266>
   rotation=<rot 7.939 -90 7.462>
  }]
 dyn_beardBraid_T3_02_l [{
   shapeType=Capsule
   length=2.3744
   radius=1.4615
   shapeName=dyn_beardBraid_T3_02_l_capsule
   center=<0 1 3.9122>
   rotation=<rot 8.8394 -90 5.2314>
  }]
 dyn_beardBraid_T3_02_r [{
   shapeType=Capsule
   length=2.3744
   radius=1.5199
   shapeName=dyn_beardBraid_T3_02_r_capsule
   center=<0 1 3.7403>
   rotation=<rot 9.6809 -90 -0.0587>
  }]
 head [{shapeType=Capsule, length=0.505, radius=0.505, center=<6.9369 2.5374 -6.1117>, rotation=<rot 0 114.6741 -64.1909>}]
 spine_04 [{shapeType=Capsule, length=5.75, radius=17.255, shapeName=spine_04_capsule, rotation=<rot 0 -90 0>}]
 spine_05 [{shapeType=Capsule, length=18.5, radius=13.38, shapeName=spine_05_capsule, center=<-0.7404 0 0>, rotation=<rot 0 -89.9999 0>}]
}
constraints [
 {bone1Name=C_jaw, bone2Name=head, swing1LimitDegrees=45, swing2LimitDegrees=45, twistLimitDegrees=45}
 {bone1Name=dyn_beardBraid_02_a_c, bone2Name=dyn_beardBraid_01_c, swing1LimitDegrees=45, swing2LimitDegrees=45, twistLimitDegrees=15}
 {bone1Name=dyn_beardBraid_02_b_c, bone2Name=dyn_beardBraid_02_a_c, swing1LimitDegrees=45, swing2LimitDegrees=45, twistLimitDegrees=15}
 {bone1Name=dyn_beardBraid_T3_01_r, bone2Name=C_jaw, swing1LimitDegrees=20, swing2LimitDegrees=180, twistLimitDegrees=30}
 {bone1Name=dyn_beardBraid_T3_02_r, bone2Name=dyn_beardBraid_T3_01_r, swing1LimitDegrees=15, swing2LimitDegrees=180, twistLimitDegrees=30}
 {
  bone1Name=dyn_beardBraid_01_c
  bone2Name=C_jaw
  swing1LimitDegrees=45
  swing1Motion=Free
  swing2LimitDegrees=45
  swing2Motion=Free
  twistLimitDegrees=45
  twistMotion=Free
 }
 {bone1Name=dyn_beardBraid_T3_01_l, bone2Name=C_jaw, swing1LimitDegrees=20, swing2LimitDegrees=180, twistLimitDegrees=30}
 {bone1Name=dyn_beardBraid_T3_02_l, bone2Name=dyn_beardBraid_T3_01_l, swing1LimitDegrees=15, swing2LimitDegrees=180, twistLimitDegrees=30}
 {
  bone1Name=clavicle_l
  bone2Name=spine_05
  swing1LimitDegrees=45
  swing1Motion=Free
  swing2LimitDegrees=45
  swing2Motion=Free
  twistLimitDegrees=45
  twistMotion=Free
 }
 {
  bone1Name=clavicle_r
  bone2Name=spine_05
  swing1LimitDegrees=45
  swing1Motion=Free
  swing2LimitDegrees=45
  swing2Motion=Free
  twistLimitDegrees=45
  twistMotion=Free
 }
 {
  bone1Name=head
  bone2Name=spine_05
  swing1LimitDegrees=45
  swing1Motion=Free
  swing2LimitDegrees=45
  swing2Motion=Free
  twistLimitDegrees=45
  twistMotion=Free
 }
 {
  bone1Name=spine_05
  bone2Name=spine_04
  swing1LimitDegrees=45
  swing1Motion=Free
  swing2LimitDegrees=45
  swing2Motion=Free
  twistLimitDegrees=45
  twistMotion=Free
 }
]
