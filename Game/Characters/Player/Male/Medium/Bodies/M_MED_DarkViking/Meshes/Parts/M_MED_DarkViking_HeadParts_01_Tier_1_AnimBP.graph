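# Anim Blueprint for the Dark Viking Tier 1 head parts, adding secondary motion to beard braids and hair.
# It copies the attached parent's pose into a cache, then chains per-bone AnimDynamics and a RigidBody simulation, with beard spherical limits on spine_05.
# A playlist custom-character-parts flag, read at initialization, instead selects a pose with the head bone scaled to near zero.
# Two beard AnimDynamics nodes and the MeshToCopy setup are unconnected.
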AnimBlueprint /Game/Characters/Player/Male/Medium/Bodies/M_MED_DarkViking/Meshes/Parts/M_MED_DarkViking_HeadParts_01_Tier_1_AnimBP
target_skeleton=&/M_MED_DarkViking_HeadParts_01_Tier_1_Skeleton
graphs [
 AnimGraph {nodes [
   Anim Dynamics - Bone: dyn_beardBraid_01_c(Alpha=1, ExternalForce="0,0,0") {
    angularDampingOverride=0.99
    angularSpringConstant=500
    bAngularSpring=true
    bGravityOverrideInSimSpace=true
    bOverrideAngularDamping=true
    bUseSphericalLimits=true
    gravityScale=0.2
    alphaBoolBlend {blendOption=Linear}
    boundBone {boneName=dyn_beardBraid_01_c}
    chainEnd {boneName=dyn_beardBraid_03_c}
    physicsBodyDefinitions [{
      collisionType=CustomSphere
      sphereCollisionRadius=1
      boundBone {boneName=dyn_beardBraid_01_c}
      boxExtents=<5 20 20>
      constraintSetup {angularLimitsMax=<0 20 30>, angularLimitsMin=<0 -20 -20>, angularTarget=<0 0 0>}
      localJointOffset=<5 0 0>
     }]
    relativeSpaceBone {boneName=C_jaw}
    sphericalLimits [{limitRadius=15, drivingBone {boneName=spine_05}, sphereLocalOffset=<0 0 0>}]
   }
   "Anim Dynamics - Chain: dyn_beardBraid_01_c -> dyn_beardBraid_03_c"(Alpha=1, ExternalForce="0,0,0") {
    angularDampingOverride=0.99
    angularSpringConstant=500
    bAngularSpring=true
    bChain=true
    bGravityOverrideInSimSpace=true
    bOverrideAngularDamping=true
    bUseSphericalLimits=true
    gravityScale=0.5
    linearSpringConstant=1200
    alphaBoolBlend {blendOption=Linear}
    boundBone {boneName=dyn_beardBraid_01_c}
    chainEnd {boneName=dyn_beardBraid_03_c}
    physicsBodyDefinitions [{
      collisionType=CustomSphere
      sphereCollisionRadius=1
      boundBone {boneName=dyn_beardBraid_01_c}
      boxExtents=<5 20 20>
      constraintSetup {angularLimitsMax=<0 20 30>, angularLimitsMin=<0 -20 -20>, angularTarget=<0 0 0>}
      localJointOffset=<2 0 0>
     }]
    relativeSpaceBone {boneName=C_jaw}
    sphericalLimits [{limitRadius=15, drivingBone {boneName=spine_05}, sphereLocalOffset=<0 0 0>}]
   }
   n1 = "Transform (Modify) Bone - Bone: head"!(ComponentPose=Local To Component(LocalPose=Use cached pose 'PreScale Cache'()), Alpha=1) -> (Pose) {scaleMode=BMM_Replace, alphaBoolBlend {blendOption=Linear}, boneToModify {boneName=head}, scale=<0.01 0.01 0.01>}
   n2 = Anim Dynamics - Bone: dyn_beardBraid_02_a_c(ComponentPose=Local To Component(LocalPose=Use cached pose 'PreScale Cache'()), Alpha=1, ExternalForce="0,0,0") {
    angularDampingOverride=0.99
    angularSpringConstant=500
    bAngularSpring=true
    bGravityOverrideInSimSpace=true
    bOverrideAngularDamping=true
    bUseSphericalLimits=true
    gravityScale=0.2
    alphaBoolBlend {blendOption=Linear}
    boundBone {boneName=dyn_beardBraid_02_a_c}
    physicsBodyDefinitions [{
      collisionType=CustomSphere
      sphereCollisionRadius=3
      boundBone {boneName=dyn_beardBraid_02_a_c}
      boxExtents=<5 20 20>
      constraintSetup {angularLimitsMax=<0 20 30>, angularLimitsMin=<0 -20 -10>, angularTarget=<0 0 0>}
      localJointOffset=<5 0 0>
     }]
    relativeSpaceBone {boneName=C_jaw}
    sphericalLimits [{limitRadius=15, drivingBone {boneName=spine_05}, sphereLocalOffset=<0 0 0>}]
   }
   n3 = Anim Dynamics - Bone: dyn_beardBraid_02_b_c(ComponentPose=@n2, Alpha=1, ExternalForce="0,0,0") {
    angularDampingOverride=0.99
    angularSpringConstant=500
    bAngularSpring=true
    bGravityOverrideInSimSpace=true
    bOverrideAngularDamping=true
    bUseSphericalLimits=true
    gravityScale=0.2
    alphaBoolBlend {blendOption=Linear}
    boundBone {boneName=dyn_beardBraid_02_b_c}
    physicsBodyDefinitions [{
      collisionType=CustomSphere
      sphereCollisionRadius=1
      boundBone {boneName=dyn_beardBraid_02_b_c}
      boxExtents=<5 20 20>
      constraintSetup {angularLimitsMax=<0 20 30>, angularLimitsMin=<0 -20 -20>, angularTarget=<0 0 0>}
      localJointOffset=<5 0 0>
     }]
    relativeSpaceBone {boneName=C_jaw}
    sphericalLimits [{limitRadius=15, drivingBone {boneName=spine_05}, sphereLocalOffset=<0 0 0>}]
   }
   n4 = Anim Dynamics - Bone: dyn_beardBraid_T3_01_l(ComponentPose=@n3, Alpha=1, ExternalForce="0,0,0") {
    angularDampingOverride=0.99
    angularSpringConstant=500
    bAngularSpring=true
    bGravityOverrideInSimSpace=true
    bOverrideAngularDamping=true
    bUseSphericalLimits=true
    gravityScale=0.2
    alphaBoolBlend {blendOption=Linear}
    boundBone {boneName=dyn_beardBraid_T3_01_l}
    physicsBodyDefinitions [{
      collisionType=CustomSphere
      sphereCollisionRadius=1
      boundBone {boneName=dyn_beardBraid_T3_01_l}
      boxExtents=<20 20 5>
      constraintSetup {angularLimitsMax=<0 20 30>, angularLimitsMin=<0 -20 -20>, angularTarget=<0 0 0>}
      localJointOffset=<0 0 10>
     }]
    sphericalLimits [{limitRadius=14, drivingBone {boneName=spine_05}, sphereLocalOffset=<0 0 0>}]
   }
   n5 = Anim Dynamics - Bone: dyn_beardBraid_T3_02_l(ComponentPose=@n4, Alpha=1, ExternalForce="0,0,0") {
    angularDampingOverride=0.99
    angularSpringConstant=500
    bAngularSpring=true
    bGravityOverrideInSimSpace=true
    bOverrideAngularDamping=true
    bUseSphericalLimits=true
    gravityScale=0.2
    alphaBoolBlend {blendOption=Linear}
    boundBone {boneName=dyn_beardBraid_T3_02_l}
    physicsBodyDefinitions [{
      collisionType=CustomSphere
      sphereCollisionRadius=1
      boundBone {boneName=dyn_beardBraid_T3_02_l}
      boxExtents=<20 20 5>
      constraintSetup {angularLimitsMax=<0 20 30>, angularLimitsMin=<0 -20 -20>, angularTarget=<0 0 0>}
      localJointOffset=<0 0 10>
     }]
    sphericalLimits [{limitRadius=14, drivingBone {boneName=spine_05}, sphereLocalOffset=<0 0 0>}]
   }
   n6 = Anim Dynamics - Bone: dyn_beardBraid_T3_01_r(ComponentPose=@n5, Alpha=1, ExternalForce="0,0,0") {
    angularDampingOverride=0.99
    angularSpringConstant=500
    bAngularSpring=true
    bGravityOverrideInSimSpace=true
    bOverrideAngularDamping=true
    bUseSphericalLimits=true
    gravityScale=0.2
    alphaBoolBlend {blendOption=Linear}
    boundBone {boneName=dyn_beardBraid_T3_01_r}
    physicsBodyDefinitions [{
      collisionType=CustomSphere
      sphereCollisionRadius=1
      boundBone {boneName=dyn_beardBraid_T3_01_r}
      boxExtents=<20 20 5>
      constraintSetup {angularLimitsMax=<0 20 30>, angularLimitsMin=<0 -20 -20>, angularTarget=<0 0 0>}
      localJointOffset=<0 0 10>
     }]
    sphericalLimits [{limitRadius=12, drivingBone {boneName=spine_05}, sphereLocalOffset=<0 0 0>}]
   }
   n7 = Anim Dynamics - Bone: dyn_beardBraid_T3_02_r(ComponentPose=@n6, Alpha=1, ExternalForce="0,0,0") {
    angularDampingOverride=0.99
    angularSpringConstant=500
    bAngularSpring=true
    bGravityOverrideInSimSpace=true
    bOverrideAngularDamping=true
    bUseSphericalLimits=true
    gravityScale=0.2
    alphaBoolBlend {blendOption=Linear}
    boundBone {boneName=dyn_beardBraid_T3_02_r}
    physicsBodyDefinitions [{
      collisionType=CustomSphere
      sphereCollisionRadius=1
      boundBone {boneName=dyn_beardBraid_T3_02_r}
      boxExtents=<20 20 5>
      constraintSetup {angularLimitsMax=<0 20 30>, angularLimitsMin=<0 -20 -20>, angularTarget=<0 0 0>}
      localJointOffset=<0 0 10>
     }]
    sphericalLimits [{limitRadius=12, drivingBone {boneName=spine_05}, sphereLocalOffset=<0 0 0>}]
   }
   n8 = RigidBody!(
    ComponentPose=@n7
    Alpha
    OverrideWorldGravity:Vector="0,0,-980"
    ExternalForce
    ComponentLinearAccScale:Vector="0.3,0.3,0.3"
    ComponentLinearVelScale:Vector="0.25,0.25,0.25"
   ) -> (Pose) {
    bClampLinearTranslationLimitToRefPose=true
    bOverrideWorldGravity=true
    overridePhysicsAsset=&/M_MED_DarkViking_HeadParts_01_Tier_1_Physics
    simulationSpace=BaseBoneSpace
    alphaBoolBlend {blendOption=Linear}
    baseBoneRef {boneName=spine_05}
    componentAppliedLinearAccClamp=<3000 3000 3000>
    componentLinearAccScale=<0.4 0.4 0.6>
    componentLinearVelScale=<0.4 0.4 0.6>
   }
   n9 = Anim Dynamics - Bone: dyn_hair_01(ComponentPose=@n8, Alpha=1) {
    angularBiasOverride=1
    angularDampingOverride=0.99
    bGravityOverrideInSimSpace=true
    bLinearSpring=true
    bOverrideAngularBias=true
    bOverrideAngularDamping=true
    bOverrideLinearDamping=true
    gravityScale=0.1
    linearDampingOverride=0.995
    linearSpringConstant=200
    alphaBoolBlend {blendOption=Linear}
    boundBone {boneName=dyn_hair_01}
    componentLinearAccScale=<1 1 1>
    componentLinearVelScale=<1 1 1>
    physicsBodyDefinitions [{
      sphereCollisionRadius=0
      boundBone {boneName=dyn_hair_01}
      boxExtents=<5 5 5>
      constraintSetup {
       angularTargetAxis=AxisY
       twistAxis=AxisY
       angularLimitsMax=<0.001 10 10>
       angularLimitsMin=<-0.001 -10 -10>
       angularTarget=<0 0 0>
       linearAxesMax=<0 0 3>
      }
      localJointOffset=<0 0 -2>
     }]
   }
   n10 = Anim Dynamics - Bone: dyn_hair_03(ComponentPose=@n9, Alpha=1, ExternalForce="0,0,0") {
    angularDampingOverride=0.99
    angularSpringConstant=1200
    bAngularSpring=true
    bGravityOverrideInSimSpace=true
    bOverrideAngularDamping=true
    alphaBoolBlend {blendOption=Linear}
    boundBone {boneName=dyn_hair_03}
    componentLinearAccScale=<1 1 1>
    componentLinearVelScale=<1 1 1>
    physicsBodyDefinitions [{
      sphereCollisionRadius=0
      boundBone {boneName=dyn_hair_03}
      boxExtents=<5 20 20>
      constraintSetup {angularLimitsMax=<0 30 0>, angularLimitsMin=<0 -10 0>, angularTarget=<0 0 0>}
      localJointOffset=<10 0 0>
     }]
   }
   n11 = Anim Dynamics - Bone: dyn_hair_02(ComponentPose=@n10, Alpha=1) {
    angularDampingOverride=0.99
    angularSpringConstant=1200
    bAngularSpring=true
    bGravityOverrideInSimSpace=true
    bOverrideAngularDamping=true
    alphaBoolBlend {blendOption=Linear}
    boundBone {boneName=dyn_hair_02}
    componentLinearAccScale=<1 1 1>
    componentLinearVelScale=<1 1 1>
    physicsBodyDefinitions [{
      sphereCollisionRadius=0
      boundBone {boneName=dyn_hair_02}
      boxExtents=<5 20 20>
      constraintSetup {angularLimitsMax=<0 30 0>, angularLimitsMin=<0 -10 0>, angularTarget=<0 0 0>}
      localJointOffset=<10 0 0>
     }]
   }
   Output Pose(Result=Blend Poses by bool(
     BlendPose_0=Component To Local(ComponentPose=@n1)
     BlendPose_1=Component To Local(ComponentPose=@n11)
     BlendTime_0=0
     BlendTime_1=0
     bActiveValue=Get GetPlayerListUsesCustomCharacterParts() -> (GetPlayerListUsesCustomCharacterParts:bool)
    ) {blendType=Linear}) {layerGroup=None}
   Save cached pose 'PreScale Cache'(Pose=Copy Pose From Mesh() {bUseAttachedParent=true}) {cacheName=PreScale Cache}
  ]}
 EventGraph {nodes [
   n1 = Cast To SkeletalMeshComponent!(
    execute
    Object=GetChildComponent!(self:Scene Component Object Reference=GetOwnerSkeletalMeshComponent(), ChildIndex:int=1) -> (ReturnValue:Scene Component Object Reference)
   ) -> (then, CastFailed, AsSkeletal Mesh Component:Skeletal Mesh Component Object Reference) {is=title K2Node_DynamicCast}
   > Set MeshToCopy(MeshToCopy:Skeletal Mesh Component Object Reference=@n1) -> (Output_Get:Skeletal Mesh Component Object Reference)
   Event BlueprintInitializeAnimation()
   > Set GetPlayerListUsesCustomCharacterParts(GetPlayerListUsesCustomCharacterParts:bool=GetPlaylistUsesCustomCharacterParts!() -> (ReturnValue:bool)) -> (Output_Get:bool)
   TryGetPawnOwner()
   Event BlueprintUpdateAnimation()
  ]}
]
variables [
 GetPlayerListUsesCustomCharacterParts {type=boolean}
 MeshToCopy {type=SkeletalMeshComponent}
]
