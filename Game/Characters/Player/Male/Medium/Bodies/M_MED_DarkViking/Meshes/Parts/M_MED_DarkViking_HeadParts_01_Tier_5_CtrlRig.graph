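# Control Rig for the Tier 5 Dark Viking head parts that applies pose-driven correctives to the beard braid base bones.
# A spherical pose reader on the head weights offsets for the left, right and center braids, named for a lowered head.
# Two further readers on neck_01, with differing rotation offsets, each weight an extra offset on one side braid.
# Offset transforms come from variables, and all three readers have debug drawing enabled.
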ControlRigBlueprint /Game/Characters/Player/Male/Medium/Bodies/M_MED_DarkViking/Meshes/Parts/M_MED_DarkViking_HeadParts_01_Tier_5_CtrlRig
graphs [
 ContainedGraph {nodes [
   n1 = RigVMFunctionEntryNode()
   SequenceAggregate = &RigVMFunction_Sequence(ExecuteContext=@n1.ExecuteContext)
   SequenceAggregate_1 = &RigVMFunction_Sequence(ExecuteContext=@SequenceAggregate.B)
   RigVMFunctionReturnNode(A=@SequenceAggregate.A, B=@SequenceAggregate_1.A, C=@SequenceAggregate_1.B)
  ]}
 RigVMFunctionLibrary {links [], nodes []}
 RigVMModel {nodes [
   RigUnit_BeginExecution = &RigUnit_BeginExecution()
   SequenceAggregate = RigVMAggregateNode(ExecuteContext=@RigUnit_BeginExecution.ExecutePin)
   n1 = &RigUnit_SphericalPoseReader(
    ActiveRegionScaleFactors=(PositiveWidth=1,NegativeWidth=1,PositiveHeight=1,NegativeHeight=1)
    ActiveRegionSize=0.11
    Debug=(bDrawDebug=True,DebugScale=25,DebugSegments=20,DebugThickness=0.25)
    DriverAxis=(X=1,Y=0,Z=0)
    DriverItem=(Type=Bone,Name="head")
    FalloffRegionScaleFactors=(PositiveWidth=1,NegativeWidth=1,PositiveHeight=1,NegativeHeight=1)
    FalloffSize=0.165
    OptionalParentItem=(Type=Bone)
    RotationOffset=(X=90,Y=0,Z=-60)
    ExecutePin=@SequenceAggregate.A
   )
   > &RigUnit_ModifyTransforms(
    ItemToModify=((Item=(Type=Bone,Name="base_beardBraid_01_l"),Transform=(Rotation=(X=0,Y=0,Z=0,W=1),Translation=(X=0,Y=0,Z=0),Scale3D=(X=1,Y=1,Z=1))))
    ItemToModify.0.Transform=RigVMVariableNode() {variable=HeadLow_L_Beard}.Value
    Weight=@n1.OutputParam
   )
   > &RigUnit_ModifyTransforms(
    ItemToModify=((Item=(Type=Bone,Name="base_beardBraid_01_r"),Transform=(Rotation=(X=0,Y=0,Z=0,W=1),Translation=(X=0,Y=0,Z=0),Scale3D=(X=1,Y=1,Z=1))))
    ItemToModify.0.Transform=RigVMVariableNode() {variable=HeadLow_R_Beard}.Value
    Weight=@n1.OutputParam
   )
   > &RigUnit_ModifyTransforms(
    ItemToModify=((Item=(Type=Bone,Name="base_beardBraid_02_c"),Transform=(Rotation=(X=0,Y=0,Z=0,W=1),Translation=(X=0,Y=0,Z=0),Scale3D=(X=1,Y=1,Z=1))))
    ItemToModify.0.Transform=RigVMVariableNode() {variable=HeadLow_C_Beard}.Value
    Weight=@n1.OutputParam
   )
   n2 = &RigUnit_SphericalPoseReader(
    ActiveRegionScaleFactors=(PositiveWidth=1,NegativeWidth=1,PositiveHeight=1,NegativeHeight=1)
    ActiveRegionSize=0.04
    Debug=(bDrawDebug=True,DebugScale=25,DebugSegments=20,DebugThickness=0.25)
    DriverAxis=(X=0,Y=1,Z=0)
    DriverItem=(Type=Bone,Name="neck_01")
    FalloffRegionScaleFactors=(PositiveWidth=1,NegativeWidth=1,PositiveHeight=1,NegativeHeight=1)
    FalloffSize=0.06
    OptionalParentItem=(Type=Bone)
    RotationOffset=(X=102,Y=0,Z=0)
    ExecutePin=@SequenceAggregate.B
   )
   > &RigUnit_ModifyTransforms(
    ItemToModify=((Item=(Type=Bone,Name="base_beardBraid_01_r"),Transform=(Rotation=(X=0,Y=0,Z=0,W=1),Translation=(X=0,Y=0,Z=0),Scale3D=(X=1,Y=1,Z=1))))
    Weight=@n2.OutputParam
    ItemToModify.0.Transform=RigVMVariableNode() {variable=Head_L_R_Bead}.Value
   )
   n3 = &RigUnit_SphericalPoseReader(
    ActiveRegionScaleFactors=(PositiveWidth=1,NegativeWidth=1,PositiveHeight=1,NegativeHeight=1)
    ActiveRegionSize=0.04
    Debug=(bDrawDebug=True,DebugScale=25,DebugSegments=20,DebugThickness=0.25)
    DriverAxis=(X=0,Y=1,Z=0)
    DriverItem=(Type=Bone,Name="neck_01")
    FalloffRegionScaleFactors=(PositiveWidth=1,NegativeWidth=1,PositiveHeight=1,NegativeHeight=1)
    FalloffSize=0.06
    OptionalParentItem=(Type=Bone)
    RotationOffset=(X=78,Y=0,Z=0)
    ExecutePin=@SequenceAggregate.C
   )
   > &RigUnit_ModifyTransforms(
    ItemToModify=((Item=(Type=Bone,Name="base_beardBraid_01_l"),Transform=(Rotation=(X=0,Y=0,Z=0,W=1),Translation=(X=0,Y=0,Z=0),Scale3D=(X=1,Y=1,Z=1))))
    Weight=@n3.OutputParam
    ItemToModify.0.Transform=RigVMVariableNode() {variable=Head_R_L_Bead}.Value
   )
  ]}
]
hierarchy [
 root {loc=<0 0 0>, rot=<0 0 0 1>}
 attach {parent=root, loc=<0 0 0>, rot=<0 0 0 1>}
 pelvis {parent=root, loc=<0 -0.7071 93.1191>, rot=<0.0127 0.707 -0.0127 -0.707>}
 dyn_pouch_r {parent=pelvis, loc=<0 0 0>, rot=<0.7071 -0.0013 -0.7071 -0.0013>}
 spine_01 {parent=pelvis, loc=<6.7494 -0.6388 0>, rot=<0 0 0.034 -0.9994>}
 spine_02 {parent=spine_01, loc=<7.2518 -0.035 0>, rot=<0 0 0.0238 -0.9997>}
 spine_03 {parent=spine_02, loc=<6.0104 0 0>, rot=<0 0 0.0149 -0.9999>}
 spine_04 {parent=spine_03, loc=<8.1024 0 0>, rot=<0 0 -0.0132 -0.9999>}
 spine_05 {parent=spine_04, loc=<10.0146 0 0>, rot=<0 0 -0.0015 -1>}
 clavicle_l {parent=spine_05, loc=<4.3575 -0.1466 -2.3223>, rot=<0.0465 -0.7289 -0.009 -0.6829>}
 clavicle_r {parent=spine_05, loc=<4.3575 -0.1466 2.3223>, rot=<0.7289 0.0465 -0.6829 0.009>}
 neck_01 {parent=spine_05, loc=<11.3138 0.1529 0>, rot=<0 0 -0.2003 -0.9797>}
 neck_02 {parent=neck_01, loc=<4.5627 0 0>, rot=<0 0 0.0504 -0.9987>}
 head {parent=neck_02, loc=<5.3914 0 0>, rot=<0 0 0.115 -0.9934>}
 faceAttach {parent=head, loc=<-0.8266 -0.1832 0>, rot=<0.0044 -0.3827 -0.0044 -0.9239>, scale=<0.89 0.89 0.89>}
 C_brow_mid {parent=faceAttach, loc=<4.7209 11.2084 4.8303>, rot=<-0.0586 -0.317 0.0114 -0.9466>}
 L_brow_mid {parent=faceAttach, loc=<7.639 10.371 3.157>, rot=<0.0603 -0.3852 0.2489 -0.8866>}
 L_brow_outer {parent=faceAttach, loc=<9.1042 8.1218 1.4857>, rot=<0.2349 -0.2525 0.3549 -0.869>}
 L_cheek_inner {parent=faceAttach, loc=<4.2785 8.761 -1.4435>, rot=<0.0454 -0.6262 0.3191 -0.7099>}
 C_lip_upper_mid {parent=faceAttach, loc=<0.0213 11.1151 0.0938>, rot=<0.2097 -0.3713 -0.0897 -0.9001>}
 L_lip_upper_outer {parent=faceAttach, loc=<1.2443 10.2873 -1.2117>, rot=<0.138 -0.435 0.2257 -0.8607>}
 L_lip_corner {parent=faceAttach, loc=<1.9818 8.8407 -2.662>, rot=<0.0012 -0.375 -0.0189 -0.9269>}
 C_nose_bridge {parent=faceAttach, loc=<2.3602 13.1515 2.4853>, rot=<-0.2002 -0.3676 0.0825 -0.9044>}
 R_brow_mid {parent=faceAttach, loc=<3.0903 10.342 7.7132>, rot=<-0.1347 -0.3556 -0.2231 -0.8976>}
 R_brow_outer {parent=faceAttach, loc=<1.412 8.0675 9.1444>, rot=<-0.1251 -0.3913 -0.4176 -0.8104>}
 R_cheek_inner {parent=faceAttach, loc=<-1.4992 8.7305 4.3237>, rot=<-0.1979 -0.0545 -0.2598 -0.9436>}
 R_lip_corner {parent=faceAttach, loc=<-2.6834 8.8629 2.0022>, rot=<0.0127 -0.39 0.008 -0.9207>}
 C_jaw {parent=faceAttach, loc=<0.3123 3.2483 0.3582>, rot=<0.0044 -0.3827 -0.0044 -0.9239>}
 L_lip_lower_outer {parent=C_jaw, loc=<1.834 6.9616 -1.2252>, rot=<0.0171 0.0671 0.3226 -0.944>}
 C_lip_lower_mid {parent=C_jaw, loc=<0.0295 7.663 -1.3214>, rot=<0 -0.0099 -0.0021 -0.9999>}
 R_lip_lower_outer {parent=C_jaw, loc=<-1.7935 6.9675 -1.2225>, rot=<0.0152 -0.0567 -0.3148 -0.9473>}
 teeth_lower {parent=C_jaw, loc=<0.0169 5.3598 -2.8149>, rot=<0 0 0 1>}
 tongue {parent=C_jaw, loc=<0.0178 1.5245 -1.0398>, rot=<-0.1568 0 0 -0.9876>}
 dyn_beardBraid_01_c {parent=C_jaw, loc=<0.0595 6.1599 -4.8061>, rot=<-0.3216 -0.6355 -0.317 -0.6264>, scale=<1.1236 1.1236 1.1236>}
 dyn_beardBraid_02_a_c {parent=dyn_beardBraid_01_c, loc=<3.1737 -0.5896 -0.0039>, rot=<0.0047 0.0055 0.2268 -0.9739>}
 dyn_beardBraid_02_b_c {parent=dyn_beardBraid_02_a_c, loc=<6.4547 0 0>, rot=<0 0 0.0435 -0.9991>}
 base_beardBraid_02_c {parent=dyn_beardBraid_01_c, loc=<4.8159 1.0505 -0.0282>, rot=<0.0058 0.0064 0.1948 -0.9808>}
 dyn_beardBraid_02_c {parent=base_beardBraid_02_c, loc=<0 0 0>, rot=<0 0 0 1>}
 dyn_beardBraid_03_c {parent=dyn_beardBraid_02_c, loc=<6.5743 0.0973 0.0001>, rot=<0 0 0 1>}
 dyn_beardBraid_T3_01_l {parent=C_jaw, loc=<4.0977 6.2492 -5.1292>, rot=<-0.1274 -0.9907 -0.0064 -0.0465>, scale=<1.1236 1.1236 1.1236>}
 dyn_beardBraid_T3_02_l {parent=dyn_beardBraid_T3_01_l, loc=<0.3289 1.2645 4.578>, rot=<-0.0078 -0.0354 0.0468 -0.9982>}
 dyn_beardBraid_T3_01_r {parent=C_jaw, loc=<-4.2875 6.2514 -5.2756>, rot=<-0.1275 -0.9914 0.004 0.0292>, scale=<1.1236 1.1236 1.1236>}
 dyn_beardBraid_T3_02_r {parent=dyn_beardBraid_T3_01_r, loc=<0.3332 1.2695 4.5766>, rot=<0.0079 0.0357 0.0467 -0.9982>}
 R_eye {parent=faceAttach, loc=<2.2537 8.4351 6.6885>, rot=<0.3823 -0.0165 0.9223 -0.0548>}
 R_eye_lid_upper_mid {parent=R_eye, loc=<-0.0431 -1.1701 0.2679>, rot=<0.0541 -0.001 0.0182 -0.9984>}
 R_eye_lid_lower_mid {parent=R_eye, loc=<-0.0113 -0.9436 -0.7026>, rot=<0.0339 -0.0006 0.0182 -0.9993>}
 L_eye {parent=faceAttach, loc=<6.6314 8.4647 2.3108>, rot=<0.3819 0.0253 0.9227 0.046>}
 L_eye_lid_upper_mid {parent=L_eye, loc=<0.0251 -1.1623 0.2388>, rot=<0.0523 0 0 -0.9986>}
 L_eye_lid_lower_mid {parent=L_eye, loc=<0.0223 -0.923 -0.7488>, rot=<0.0329 0 0 -0.9995>}
 teeth_upper {parent=faceAttach, loc=<0.0954 9.3381 0.1585>, rot=<0.0044 -0.3827 -0.0044 -0.9239>}
 R_lip_upper_outer {parent=faceAttach, loc=<-1.2731 10.2813 1.3051>, rot=<-0.0654 -0.2976 -0.2619 -0.9157>}
 dyn_helmet_rear01 {parent=head, loc=<4.3432 1.7829 0>, rot=<0.0029 -0.7071 -0.0029 -0.7071>}
 dyn_helmet_rear02 {parent=head, loc=<3.4454 1.7992 0>, rot=<0.0029 -0.7071 -0.0029 -0.7071>}
 dyn_helmet_rear03 {parent=head, loc=<0.8658 2.3142 0>, rot=<0.0029 -0.7071 -0.0029 -0.7071>}
 dyn_jawbone_01 {parent=head, loc=<-0.6062 4.7913 -0.0048>, rot=<0.0041 -1 0 0>}
 dyn_jawbone_02 {parent=dyn_jawbone_01, loc=<0 0 0>, rot=<0 0 0 1>}
 dyn_hair_01 {parent=head, loc=<14.2123 4.1721 -3.6492>, rot=<0.0029 -0.7071 -0.0029 -0.7071>}
 dyn_hair_03 {parent=head, loc=<7.2242 4.4471 -7.9422>, rot=<0.0041 -1 0 0>}
 dyn_hair_02 {parent=head, loc=<9.2305 -0.1516 -7.662>, rot=<0.1429 -0.9338 0.3246 0.0469>}
 base_beardBraid_01_l {parent=head, loc=<-5.5519 10.0828 -3.4193>, rot=<-0.1882 -0.9821 -0.0023 -0.0115>}
 dyn_beardBraid_01_l {parent=base_beardBraid_01_l, loc=<0 0 0>, rot=<0 0 0 1>}
 dyn_beardBraid_02_l {parent=dyn_beardBraid_01_l, loc=<5.684 0 0>, rot=<0 0 0 1>}
 base_beardBraid_01_r {parent=head, loc=<-5.5519 10.0828 3.4193>, rot=<-0.1882 -0.9821 0.0023 0.0115>}
 dyn_beardBraid_01_r {parent=base_beardBraid_01_r, loc=<0 0 0>, rot=<0 0 0 1>}
 dyn_beardBraid_02_r {parent=dyn_beardBraid_01_r, loc=<5.684 0 0>, rot=<0 0 0 1>}
 dyn_beard_01 {parent=spine_05, loc=<10.6697 11.8295 0>, rot=<0.0284 -0.7065 -0.0284 -0.7065>}
 dyn_chestStrap_05 {parent=spine_05, loc=<0 0 0>, rot=<0.0284 -0.7065 -0.0284 -0.7065>}
 dyn_neck_hair {parent=spine_05, loc=<12.1424 -6.4889 0>, rot=<0.0284 -0.7065 -0.0284 -0.7065>}
 dyn_nude_chestStrap_05 {parent=spine_05, loc=<0 0 0>, rot=<0.0284 -0.7065 -0.0284 -0.7065>}
 dyn_nude_chestStrap_front_05 {parent=spine_05, loc=<0 0 0>, rot=<0.0284 -0.7065 -0.0284 -0.7065>}
 dyn_chestStrap_04 {parent=spine_04, loc=<0 0 0>, rot=<0.0294 -0.7065 -0.0294 -0.7065>}
 dyn_nude_chestStrap_04 {parent=spine_04, loc=<0 0 0>, rot=<0.0294 -0.7065 -0.0294 -0.7065>}
 dyn_nude_chestStrap_front_04 {parent=spine_04, loc=<0 0 0>, rot=<0.0294 -0.7065 -0.0294 -0.7065>}
 dyn_chestStrap_03 {parent=spine_03, loc=<0 0 0>, rot=<0.0387 -0.706 -0.0387 -0.706>}
 dyn_nude_chestStrap_03 {parent=spine_03, loc=<0 0 0>, rot=<0.0387 -0.706 -0.0387 -0.706>}
 dyn_nude_chestStrap_front_03 {parent=spine_03, loc=<0 0 0>, rot=<0.0387 -0.706 -0.0387 -0.706>}
 dyn_chestStrap_02 {parent=spine_02, loc=<0 0 0>, rot=<0.0282 -0.7065 -0.0282 -0.7065>}
 dyn_nude_chestStrap_02 {parent=spine_02, loc=<0 0 0>, rot=<0.0282 -0.7065 -0.0282 -0.7065>}
 dyn_nude_chestStrap_front_02 {parent=spine_02, loc=<0 0 0>, rot=<0.0282 -0.7065 -0.0282 -0.7065>}
 dyn_chestStrap_01 {parent=spine_01, loc=<0 0 0>, rot=<0.0114 -0.707 -0.0114 -0.707>}
 dyn_nude_chestStrap_01 {parent=spine_01, loc=<0 0 0>, rot=<0.0114 -0.707 -0.0114 -0.707>}
 dyn_nude_chestStrap_front_01 {parent=spine_01, loc=<0 0 0>, rot=<0.0114 -0.707 -0.0114 -0.707>}
 dyn_horn_01 {parent=pelvis, loc=<6.0133 -5.9041 9.8187>, rot=<-0.021 -0.8829 0.4681 0.0296>}
 dyn_horn_02 {parent=dyn_horn_01, loc=<0 0 0>, rot=<0 0 0 1>}
 dyn_belt_buckle {parent=pelvis, loc=<2.6772 13.2066 -0.3301>, rot=<-0.0127 -0.707 0.0127 -0.707>}
 dyn_rear_loincloth_driver {parent=pelvis, loc=<4.1333 -9.633 -0.3727>, rot=<0 0 0 1>}
 dyn_rear_loincloth_01 {parent=dyn_rear_loincloth_driver, loc=<0 0 0>, rot=<0.0711 -0.9975 0 0>}
 dyn_rear_loincloth_02 {parent=dyn_rear_loincloth_01, loc=<11.9216 0 0>, rot=<0 0 -0.0142 -0.9999>}
 dyn_rear_loincloth_03 {parent=dyn_rear_loincloth_02, loc=<10.9003 0 0>, rot=<0 0 -0.0375 -0.9993>}
 dyn_front_loincloth_PARENT {parent=pelvis, loc=<3.0083 11.4078 0.4572>, rot=<-0.0127 -0.707 0.0127 -0.707>}
 dyn_front_loincloth_LOCAL {parent=dyn_front_loincloth_PARENT, loc=<0 0 0>, rot=<0 0 0 1>}
 dyn_front_loincloth_driver {parent=dyn_front_loincloth_LOCAL, loc=<0 0 0>, rot=<0 0 0 1>}
 dyn_front_loincloth {parent=dyn_front_loincloth_driver, loc=<0 0 0>, rot=<0 0 0 1>}
 dyn_front_loincloth_02 {parent=dyn_front_loincloth, loc=<0 2.4828 -13.088>, rot=<0 0 0 1>}
 dyn_front_loincloth_03 {parent=dyn_front_loincloth_02, loc=<0 1.3322 -9.8013>, rot=<0 0 0 1>}
 front_loin_MASTER {parent=pelvis, loc=<2.2718 10.4373 0.2011>, rot=<-0.1064 -0.9943 0.0001 0.0015>}
 front_loin_Driven_r {parent=front_loin_MASTER, loc=<1.2351 -2.7421 -9.1301>, rot=<-0.0111 -0.0736 0.0424 -0.9963>}
 front_loin_01_r {parent=front_loin_Driven_r, loc=<0 0 0>, rot=<0 0 0 1>}
 front_loin_02_r {parent=front_loin_01_r, loc=<12.9324 0 0>, rot=<0 0 0 1>}
 front_loin_03_r {parent=front_loin_02_r, loc=<21.4342 0 0>, rot=<0.0288 0.7557 0.0249 -0.6538>}
 front_loin_Driven_l {parent=front_loin_MASTER, loc=<0.21 -2.2799 8.8899>, rot=<0.0065 0.025 0.0134 -0.9996>}
 front_loin_01_l {parent=front_loin_Driven_l, loc=<0 0 0>, rot=<0 0 0 1>}
 front_loin_02_l {parent=front_loin_01_l, loc=<15.7212 0 0>, rot=<0 0 0 1>}
 front_loin_03_l {parent=front_loin_02_l, loc=<14.0773 0 0>, rot=<0.0615 0.697 0.0628 -0.7117>}
]
variables [
 HeadLow_C_Beard {type=FTransform, default=(Rotation=(X=0,Y=0,Z=0.1736,W=0.9848),Translation=(X=2,Y=0,Z=0),Scale3D=(X=1,Y=1,Z=1))}
 HeadLow_L_Beard {type=FTransform, default=(Rotation=(X=0.0302,Y=-0.171,Z=0.171,W=0.9698),Translation=(X=0,Y=2,Z=2),Scale3D=(X=1,Y=1,Z=1))}
 HeadLow_R_Beard {type=FTransform, default=(Rotation=(X=-0.0302,Y=0.171,Z=0.171,W=0.9698),Translation=(X=0,Y=2,Z=-2),Scale3D=(X=1,Y=1,Z=1))}
 Head_L_R_Bead {type=FTransform, default=(Rotation=(X=0,Y=0.0872,Z=0,W=0.9962),Translation=(X=0,Y=2,Z=-2),Scale3D=(X=1,Y=1,Z=1))}
 Head_R_L_Bead {type=FTransform, default=(Rotation=(X=0,Y=-0.0872,Z=0,W=0.9962),Translation=(X=0,Y=2,Z=2),Scale3D=(X=1,Y=1,Z=1))}
]
