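# Physics asset for the tier 5 head parts of the male medium Dark Viking player body, set up for three beard braids.
# Capsule bodies form left and right chains hanging from the head and a center chain from the jaw.
# Braid dynamic joints limit swing and nearly lock twist, while base braid, jaw and head joints rotate freely.
# Large spine_05 capsules cover the upper torso, likely as a collision surface for the braids.
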PhysicsAsset /Game/Characters/Player/Male/Medium/Bodies/M_MED_DarkViking/Meshes/Parts/M_MED_DarkViking_HeadParts_01_Tier_5_Physics
bodies {
 C_jaw [{shapeType=Capsule, length=1, radius=1, shapeName=C_jaw_capsule}]
 base_beardBraid_01_l [{shapeType=Capsule, length=0.505, radius=0.505, shapeName=base_beardBraid_01_l_capsule, rotation=<rot 0 -90 0>}]
 base_beardBraid_01_r [{shapeType=Capsule, length=0.505, radius=0.505, shapeName=base_beardBraid_01_r_capsule, rotation=<rot 0 -90 0>}]
 base_beardBraid_02_c [{shapeType=Capsule, length=0.505, radius=0.505, shapeName=base_beardBraid_02_c_capsule, rotation=<rot 0 -90 0>}]
 dyn_beardBraid_01_c [{shapeType=Capsule, length=1, radius=1, shapeName=dyn_beardBraid_01_c_capsule, center=<1 0 0>, rotation=<rot 90 0 0>}]
 dyn_beardBraid_01_l [
  {shapeType=Capsule, length=1.5, radius=1.5, shapeName=dyn_beardBraid_01_l_capsule, center=<1.5 0 0>, rotation=<rot 90 0 0>}
  {shapeType=Capsule, length=0, radius=2.25, shapeName=dyn_beardBraid_01_l_capsule, center=<3.8837 -0.1266 0>, rotation=<rot 90 0 0>}
 ]
 dyn_beardBraid_01_r [
  {shapeType=Capsule, length=1.5, radius=1.5, shapeName=dyn_beardBraid_01_r_capsule, center=<1.5 0 0>, rotation=<rot 90 0 0>}
  {shapeType=Capsule, length=0, radius=2.25, shapeName=dyn_beardBraid_01_r_capsule, center=<3.8837 -0.1266 0>, rotation=<rot 90 0 0>}
 ]
 dyn_beardBraid_02_c [{shapeType=Capsule, length=1.5, radius=1.5, shapeName=dyn_beardBraid_02_c_capsule, center=<1.5 0 0>, rotation=<rot 90 0 0>}]
 dyn_beardBraid_02_l [
  {shapeType=Capsule, length=2.5, radius=1.5, shapeName=dyn_beardBraid_02_l_capsule, center=<3.2626 0 0>, rotation=<rot 90 0 0>}
  {shapeType=Capsule, length=0, radius=1.25, shapeName=dyn_beardBraid_02_l_capsule, center=<3.6068 -0.5027 0>}
 ]
 dyn_beardBraid_02_r [
  {shapeType=Capsule, length=2.5, radius=1.5, shapeName=dyn_beardBraid_02_r_capsule, center=<3.2626 0 0>, rotation=<rot 90 0 0>}
  {shapeType=Capsule, length=0, radius=1.25, shapeName=dyn_beardBraid_02_r_capsule, center=<3.6068 -0.5027 0>}
 ]
 dyn_beardBraid_03_c [
  {shapeType=Capsule, length=1.5, radius=1.5, shapeName=dyn_beardBraid_03_c_capsule, center=<1.5 0 0>, rotation=<rot 90 0 0>}
  {
   shapeType=Capsule
   length=4.25
   radius=2.5
   shapeName=dyn_beardBraid_03_c_capsule
   center=<4.5513 -2.3197 0>
   rotation=<rot 90 180 -160.0001>
  }
 ]
 head [{shapeType=Capsule, length=1, radius=1, shapeName=head_capsule, rotation=<rot 0 -122.3545 -89.6579>}]
 neck_01 [{shapeType=Capsule, length=1, radius=1, center=<2.6657 0 0>, rotation=<rot 90 0 0>}]
 spine_05 [
  {shapeType=Capsule, length=19.3319, radius=14.9929, center=<-7.2821 0 0>, rotation=<rot 90 0 0>}
  {shapeType=Capsule, length=2.75, radius=15, shapeName=spine_05_capsule, center=<-2.0505 1.2384 0>}
 ]
}
constraints [
 {bone1Name=neck_01, bone2Name=spine_05, swing1LimitDegrees=45, swing2LimitDegrees=45, twistLimitDegrees=45}
 {bone1Name=dyn_beardBraid_01_l, bone2Name=base_beardBraid_01_l, swing1LimitDegrees=45, swing2LimitDegrees=45, twistLimitDegrees=1}
 {bone1Name=dyn_beardBraid_02_l, bone2Name=dyn_beardBraid_01_l, swing1LimitDegrees=60, swing2LimitDegrees=45, twistLimitDegrees=1}
 {bone1Name=dyn_beardBraid_01_r, bone2Name=base_beardBraid_01_r, swing1LimitDegrees=45, swing2LimitDegrees=45, twistLimitDegrees=1}
 {bone1Name=dyn_beardBraid_02_r, bone2Name=dyn_beardBraid_01_r, swing1LimitDegrees=60, swing2LimitDegrees=45, twistLimitDegrees=1}
 {bone1Name=dyn_beardBraid_02_c, bone2Name=base_beardBraid_02_c, swing1LimitDegrees=45, swing2LimitDegrees=45, twistLimitDegrees=1}
 {bone1Name=dyn_beardBraid_03_c, bone2Name=dyn_beardBraid_02_c, swing1LimitDegrees=45, swing2LimitDegrees=45, twistLimitDegrees=1}
 {
  bone1Name=C_jaw
  bone2Name=head
  swing1LimitDegrees=45
  swing1Motion=Free
  swing2LimitDegrees=45
  swing2Motion=Free
  twistLimitDegrees=45
  twistMotion=Free
 }
 {bone1Name=dyn_beardBraid_01_c, bone2Name=C_jaw, swing1LimitDegrees=45, swing2LimitDegrees=45, twistLimitDegrees=1}
 {
  bone1Name=head
  bone2Name=neck_01
  swing1LimitDegrees=45
  swing1Motion=Free
  swing2LimitDegrees=45
  swing2Motion=Free
  twistLimitDegrees=45
  twistMotion=Free
 }
 {
  bone1Name=base_beardBraid_01_r
  bone2Name=head
  swing1LimitDegrees=45
  swing1Motion=Free
  swing2LimitDegrees=45
  swing2Motion=Free
  twistLimitDegrees=45
  twistMotion=Free
 }
 {
  bone1Name=base_beardBraid_01_l
  bone2Name=head
  swing1LimitDegrees=45
  swing1Motion=Free
  swing2LimitDegrees=45
  swing2Motion=Free
  twistLimitDegrees=45
  twistMotion=Free
 }
 {
  bone1Name=base_beardBraid_02_c
  bone2Name=dyn_beardBraid_01_c
  swing1LimitDegrees=45
  swing1Motion=Free
  swing2LimitDegrees=45
  swing2Motion=Free
  twistLimitDegrees=45
  twistMotion=Free
 }
]
