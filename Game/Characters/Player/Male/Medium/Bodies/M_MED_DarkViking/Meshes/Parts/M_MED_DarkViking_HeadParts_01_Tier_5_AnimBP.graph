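# Post-process style animation Blueprint for the Tier 5 Dark Viking head parts, copying the parent mesh's pose and adding beard and hair motion.
# The live chain runs Control Rig, AnimDynamics, RigidBody, then a right-braid offset for front end outside emotes.
# When the playlist flag, read at initialization, is true, output instead shrinks the head bone to near zero.
# A per-braid AnimDynamics chain remains unconnected to the output.
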AnimBlueprint /Game/Characters/Player/Male/Medium/Bodies/M_MED_DarkViking/Meshes/Parts/M_MED_DarkViking_HeadParts_01_Tier_5_AnimBP
target_skeleton=&/M_MED_DarkViking_HeadParts_01_Tier_5_Skeleton
graphs [
 AnimGraph {nodes [
   Anim Dynamics - Bone: dyn_beardBraid_03_c(Alpha=0, ExternalForce="0,0,0") {
    angularDampingOverride=0.99
    angularSpringConstant=500
    bAngularSpring=true
    bGravityOverrideInSimSpace=true
    bOverrideAngularDamping=true
    bUseSphericalLimits=true
    gravityScale=0.2
    alphaBoolBlend {blendOption=Linear}
    boundBone {boneName=dyn_beardBraid_03_c}
    physicsBodyDefinitions [{
      collisionType=CustomSphere
      sphereCollisionRadius=1.5
      boundBone {boneName=dyn_beardBraid_03_c}
      boxExtents=<5 20 20>
      constraintSetup {angularLimitsMax=<0 20 30>, angularLimitsMin=<0 -20 -15>, angularTarget=<0 0 0>}
      localJointOffset=<5 0 0>
     }]
    relativeSpaceBone {boneName=C_jaw}
    sphericalLimits [{limitRadius=19, drivingBone {boneName=spine_05}, sphereLocalOffset=<0 0 0>}]
   }
   n1 = Anim Dynamics - Bone: dyn_beardBraid_02_c(Alpha=1, ExternalForce="0,0,0") {
    angularDampingOverride=0.99
    angularSpringConstant=500
    bAngularSpring=true
    bGravityOverrideInSimSpace=true
    bOverrideAngularDamping=true
    bUseSphericalLimits=true
    gravityScale=0.2
    alphaBoolBlend {blendOption=Linear}
    boundBone {boneName=dyn_beardBraid_02_c}
    chainEnd {boneName=dyn_beardBraid_03_c}
    physicsBodyDefinitions [
     {
      collisionType=CustomSphere
      sphereCollisionRadius=1
      boundBone {boneName=dyn_beardBraid_02_c}
      boxExtents=<5 20 20>
      constraintSetup {angularLimitsMax=<0 20 30>, angularLimitsMin=<0 -20 -40>, angularTarget=<0 0 0>}
      localJointOffset=<5 0 0>
     }
     {
      boundBone {boneName=dyn_beardBraid_03_c}
      boxExtents=<10 10 10>
      constraintSetup {angularLimitsMax=<0 0 0>, angularLimitsMin=<0 0 0>, angularTarget=<0 0 0>}
      localJointOffset=<0 0 0>
     }
    ]
    relativeSpaceBone {boneName=C_jaw}
    sphericalLimits [{limitRadius=18, drivingBone {boneName=spine_05}, sphereLocalOffset=<0 0 0>}]
   }
   n2 = Anim Dynamics - Bone: dyn_beardBraid_01_l(ComponentPose=@n1, Alpha=1, ExternalForce="0,0,0") {
    angularBiasOverride=0.99
    angularDampingOverride=0.99
    angularSpringConstant=500
    bAngularSpring=true
    bGravityOverrideInSimSpace=true
    bOverrideAngularBias=true
    bOverrideAngularDamping=true
    bUseSphericalLimits=true
    gravityScale=0.2
    alphaBoolBlend {blendOption=Linear}
    boundBone {boneName=dyn_beardBraid_01_l}
    chainEnd {boneName=dyn_beardBraid_03_c}
    physicsBodyDefinitions [{
      collisionType=CustomSphere
      sphereCollisionRadius=1.2
      boundBone {boneName=dyn_beardBraid_01_l}
      boxExtents=<5 20 20>
      constraintSetup {angularLimitsMax=<0 5 30>, angularLimitsMin=<0 -20 -10>, angularTarget=<0 0 0>}
      localJointOffset=<3 0 0>
     }]
    relativeSpaceBone {boneName=C_jaw}
    sphericalLimits [{limitRadius=19, drivingBone {boneName=spine_05}, sphereLocalOffset=<0 0 0>}]
   }
   n3 = "Transform (Modify) Bone - Bone: dyn_beardBraid_01_l"!(ComponentPose=@n2, Translation="1,0.8,-0.000", Rotation="-0.000,0.000,20") -> (Pose) {
    rotationMode=BMM_Additive
    translationMode=BMM_Additive
    alphaBoolBlend {blendOption=Linear}
    boneToModify {boneName=dyn_beardBraid_01_l}
    rotation {roll=20}
    translation=<1 0.8>
   }
   n4 = Anim Dynamics - Bone: dyn_beardBraid_02_l(ComponentPose=@n3, Alpha=1, ExternalForce="0,0,0") {
    angularDampingOverride=0.99
    angularSpringConstant=500
    bAngularSpring=true
    bGravityOverrideInSimSpace=true
    bOverrideAngularDamping=true
    bUseSphericalLimits=true
    gravityScale=0.2
    alphaBoolBlend {blendOption=Linear}
    boundBone {boneName=dyn_beardBraid_02_l}
    physicsBodyDefinitions [{
      collisionType=CustomSphere
      sphereCollisionRadius=1.2
      boundBone {boneName=dyn_beardBraid_02_l}
      boxExtents=<5 20 20>
      constraintSetup {angularLimitsMax=<0 20 30>, angularLimitsMin=<0 -20 -15>, angularTarget=<0 0 0>}
      localJointOffset=<2.5 0 0>
     }]
    relativeSpaceBone {boneName=C_jaw}
    sphericalLimits [{limitRadius=18, drivingBone {boneName=spine_05}, sphereLocalOffset=<0 0 0>}]
   }
   n5 = Anim Dynamics - Bone: dyn_beardBraid_01_r(ComponentPose=@n4, Alpha=1, ExternalForce="0,0,0") {
    angularBiasOverride=0.99
    angularDampingOverride=0.99
    angularSpringConstant=500
    bAngularSpring=true
    bGravityOverrideInSimSpace=true
    bOverrideAngularBias=true
    bOverrideAngularDamping=true
    bUseSphericalLimits=true
    gravityScale=0.2
    alphaBoolBlend {blendOption=Linear}
    boundBone {boneName=dyn_beardBraid_01_r}
    chainEnd {boneName=dyn_beardBraid_03_c}
    physicsBodyDefinitions [{
      collisionType=CustomSphere
      sphereCollisionRadius=1.2
      boundBone {boneName=dyn_beardBraid_01_r}
      boxExtents=<5 20 20>
      constraintSetup {angularLimitsMax=<0 20 30>, angularLimitsMin=<0 -5 -10>, angularTarget=<0 0 0>}
      localJointOffset=<3 0 0>
     }]
    relativeSpaceBone {boneName=C_jaw}
    sphericalLimits [{limitRadius=19, drivingBone {boneName=spine_05}, sphereLocalOffset=<0 0 0>}]
   }
   n6 = "Transform (Modify) Bone - Bone: dyn_beardBraid_01_r"!(ComponentPose=@n5, Translation="-1.000,0.8,-0.000", Rotation="-0.000,0.000,20.000") -> (Pose) {
    rotationMode=BMM_Additive
    translationMode=BMM_Additive
    alphaBoolBlend {blendOption=Linear}
    boneToModify {boneName=dyn_beardBraid_01_r}
    rotation {roll=20}
    translation=<-1 0.8>
   }
   Anim Dynamics - Bone: dyn_beardBraid_02_r(ComponentPose=@n6, Alpha=1, ExternalForce="0,0,0") {
    angularDampingOverride=0.99
    angularSpringConstant=500
    bAngularSpring=true
    bGravityOverrideInSimSpace=true
    bOverrideAngularDamping=true
    bUseSphericalLimits=true
    gravityScale=0.2
    alphaBoolBlend {blendOption=Linear}
    boundBone {boneName=dyn_beardBraid_02_r}
    physicsBodyDefinitions [{
      collisionType=CustomSphere
      sphereCollisionRadius=1.2
      boundBone {boneName=dyn_beardBraid_02_r}
      boxExtents=<5 20 20>
      constraintSetup {angularLimitsMax=<0 20 30>, angularLimitsMin=<0 -20 -15>, angularTarget=<0 0 0>}
      localJointOffset=<2.5 0 0>
     }]
    relativeSpaceBone {boneName=C_jaw}
    sphericalLimits [{limitRadius=18, drivingBone {boneName=spine_05}, sphereLocalOffset=<0 0 0>}]
   }
   n7 = "Transform (Modify) Bone - Bone: head"!(ComponentPose=Local To Component(LocalPose=Use cached pose 'PreScale Cache'()), Alpha=1) -> (Pose) {scaleMode=BMM_Replace, alphaBoolBlend {blendOption=Linear}, boneToModify {boneName=head}, scale=<0.01 0.01 0.01>}
   n8 = Control Rig(Source=Use cached pose 'PreScale Cache'()) {
    alphaBoolBlend {blendOption=Linear}
    controlRigAssetReference {blueprintRigClass=&/M_MED_DarkViking_HeadParts_01_Tier_5_CtrlRig.M_MED_DarkViking_HeadParts_01_Tier_5_CtrlRig_C}
   }
   n9 = Anim Dynamics - Bone: dyn_beard_01(ComponentPose=Local To Component(LocalPose=@n8), Alpha=0.75, ExternalForce="0,0,0") {
    bGravityOverrideInSimSpace=true
    bLinearSpring=true
    bOverrideLinearDamping=true
    gravityScale=0.1
    lODThreshold=1
    linearDampingOverride=0.995
    linearSpringConstant=200
    alphaBoolBlend {blendOption=Linear}
    boundBone {boneName=dyn_beard_01}
    componentLinearAccScale=<1 1 1>
    componentLinearVelScale=<1 1 1>
    physicsBodyDefinitions [{
      sphereCollisionRadius=0
      boundBone {boneName=dyn_beard_01}
      boxExtents=<2 2 2>
      constraintSetup {angularLimitsMax=<0 0 0>, angularLimitsMin=<0 0 0>, angularTarget=<0 0 0>, linearAxesMax=<0 0 2>, linearAxesMin=<0 0 -2>}
      localJointOffset=<0 0 -1>
     }]
   }
   n10 = Anim Dynamics - Bone: dyn_neck_hair(ComponentPose=@n9, Alpha=1, ExternalForce="0,0,0") {
    angularDampingOverride=-99
    angularSpringConstant=200
    bGravityOverrideInSimSpace=true
    bLinearSpring=true
    bOverrideAngularDamping=true
    bOverrideLinearDamping=true
    gravityScale=0.1
    lODThreshold=1
    linearDampingOverride=0.995
    linearSpringConstant=200
    alphaBoolBlend {blendOption=Linear}
    boundBone {boneName=dyn_neck_hair}
    componentLinearAccScale=<1 1 1>
    componentLinearVelScale=<1 1 1>
    physicsBodyDefinitions [{
      sphereCollisionRadius=0
      boundBone {boneName=dyn_neck_hair}
      boxExtents=<2 2 2>
      constraintSetup {angularLimitsMax=<0 0 10>, angularLimitsMin=<0 0 -10>, angularTarget=<0 0 0>, linearAxesMax=<0 0 2>, linearAxesMin=<0 0 -2>}
      localJointOffset=<0 0 -1>
     }]
   }
   n11 = RigidBody!(ComponentPose=@n10, Alpha, OverrideWorldGravity:Vector="0,0,-980", ExternalForce) -> (Pose) {
    bClampLinearTranslationLimitToRefPose=true
    bOverrideWorldGravity=true
    overridePhysicsAsset=&/M_MED_DarkViking_HeadParts_01_Tier_5_Physics
    simulationSpace=BaseBoneSpace
    alphaBoolBlend {blendOption=Linear}
    baseBoneRef {boneName=neck_02}
    componentAppliedLinearAccClamp=<3000 3000 3000>
    componentLinearAccScale=<0.4 0.3 0.3>
    componentLinearVelScale=<0.4 0.3 0.3>
    overrideWorldGravity {z=-980}
    simSpaceSettings {
     maxAngularAcceleration=1000
     maxAngularVelocity=3
     maxLinearAcceleration=700
     maxLinearVelocity=500
     velocityScaleZ=0
     worldAlpha=1
     externalLinearDragV=<0.2 0.2>
    }
   }
   n12 = Get bIsPlayingEmote()
   n13 = Get bIsInFrontEnd()
   n14 = "Transform (Modify) Bone - Bone: base_beardBraid_01_r"(
    ComponentPose=@n11
    Alpha=SelectFloat(A="To Float (Boolean)"(InBool=@n13), B=0, bPickA=NOT Boolean(A=@n12))
    Translation="0.000,1,-1"
    Rotation="-0.000,5,0.000"
    Scale="1,1,1"
   ) {
    rotationMode=BMM_Additive
    rotationSpace=BCS_BoneSpace
    translationMode=BMM_Additive
    translationSpace=BCS_BoneSpace
    alphaBoolBlend {blendOption=Linear}
    boneToModify {boneName=base_beardBraid_01_r}
    rotation {yaw=5}
    translation {y=1, z=-1}
   }
   Output Pose(Result=Blend Poses by bool(
     BlendPose_0=Component To Local(ComponentPose=@n7)
     BlendPose_1=Component To Local(ComponentPose=@n14)
     BlendTime_0=0
     BlendTime_1=0
     bActiveValue=Get GetPlayerListUsesCustomCharacterParts() -> (GetPlayerListUsesCustomCharacterParts:bool)
    ) {blendType=Linear}) {layerGroup=None}
   Save cached pose 'PreScale Cache'(Pose=Copy Pose From Mesh() {bUseAttachedParent=true}) {cacheName=PreScale Cache}
  ]}
 EventGraph {nodes [
   Event BlueprintInitializeAnimation()
   > Set GetPlayerListUsesCustomCharacterParts(GetPlayerListUsesCustomCharacterParts:bool=GetPlaylistUsesCustomCharacterParts!() -> (ReturnValue:bool)) -> (Output_Get:bool)
   TryGetPawnOwner()
   Cast To SkeletalMeshComponent!(
    execute
    Object=GetChildComponent!(self:Scene Component Object Reference=GetOwnerSkeletalMeshComponent(), ChildIndex:int=0) -> (ReturnValue:Scene Component Object Reference)
   ) -> (then, CastFailed, AsSkeletal Mesh Component:Skeletal Mesh Component Object Reference) {is=title K2Node_DynamicCast}
   Event BlueprintUpdateAnimation()
  ]}
]
variables [GetPlayerListUsesCustomCharacterParts {type=boolean}]
